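# Control Rig for the SpringBreakTrip medium male player body that applies corrective bone offsets to outfit and deform bones.
# It runs shared DeformRig_FNC (groin and glute enabled), then Hip and Knee Area functions.
# Hip offsets cloth, strap, belt and glute bones by thigh-up anim attributes, and constantly squashes the right sleeve corrective.
# Knee averages kneecaps between thigh and calf, offsetting them, pants and boot by remapped calf rotation deltas.
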ControlRigBlueprint /BRCosmetics/Characters/Player/Male/Medium/Bodies/M_MED_SpringBreakTrip/Meshes/M_MED_SpringBreakTrip_CtrlRig
graphs [
 New Function_ContainedGraph {nodes [
   n1 = RigVMFunctionEntryNode()
   RigVMFunction_Sequence = &RigVMFunction_Sequence(ExecuteContext=@n1.ExecuteContext)
   DISPATCH_RigDispatch_GetAnimAttribute = RigVMDispatchNode(BoneName=root, Default=0, Name=deformRig_l_thigh_up_alpha) {template="DISPATCH_RigDispatch_GetAnimAttribute(in Name,in BoneName,in Default,out Value,out Found)"}
   &RigUnit_ModifyTransforms(
    ItemToModify="((Item=(Type=Bone,Name=\"dyn_crossbodyCloth_base\"),Transform=(Rotation=(X=0.0915,Y=-0.4671,Z=0.2581,W=0.8407),Translation=(X=0,Y=0,Z=0),Scale3D=(X=1,Y=1,Z=1))),(Item=(Type=Bone,Name=\"dyn_legCloth_base\"),Transform=(Rotation=(X=0.256,Y=-0.0451,Z=-0.1677,W=0.9 ...(truncated)"
    WeightMinimum=0
    Weight=@DISPATCH_RigDispatch_GetAnimAttribute.Value
    ExecutePin=@RigVMFunction_Sequence.A
   )
   > &RigUnit_ModifyTransforms(
    ItemToModify="((Item=(Type=Bone,Name=\"dyn_belt_01\"),Transform=(Rotation=(X=0,Y=0,Z=0.0175,W=0.9998),Translation=(X=3.1796,Y=0,Z=0.559),Scale3D=(X=1,Y=1,Z=1))),(Item=(Type=Bone,Name=\"deform_glute_r\"),Transform=(Rotation=(X=0,Y=0,Z=0,W=1),Translation=(X=15.5712,Y=0,Z=0),S ...(truncated)"
    WeightMinimum=0
    Weight=RigVMDispatchNode(BoneName=root, Default=0, Name=deformRig_r_thigh_up_alpha) {template="DISPATCH_RigDispatch_GetAnimAttribute(in Name,in BoneName,in Default,out Value,out Found)"}.Value
   )
   > &RigUnit_ModifyTransforms(
    ItemToModify=((Item=(Type=Bone,Name="sleeve_corrective_r"),Transform=(Rotation=(X=0,Y=0,Z=0,W=1),Translation=(X=0,Y=0,Z=0),Scale3D=(X=1,Y=0.1,Z=0.1))))
    Weight=1
    WeightMinimum=0
   )
   n2 = RigVMDispatchNode(BoneName=root, Default=0, Name=deformRig_l_thigh_up_alpha) {template="DISPATCH_RigDispatch_GetAnimAttribute(in Name,in BoneName,in Default,out Value,out Found)"}
   > &RigUnit_ModifyTransforms(
    ItemToModify=((Item=(Type=Bone,Name="dyn_legCloth_base"),Transform=(Rotation=(X=-0.0773,Y=0.0402,Z=-0.46,W=0.8836),Translation=(X=0.5,Y=0,Z=1.5),Scale3D=(X=1,Y=1,Z=1))))
    WeightMinimum=0
    Weight=@n2.Value
   )
   > &RigUnit_ModifyTransforms(
    ItemToModify=((Item=(Type=Bone,Name="dyn_crossbodyStrap_base"),Transform=(Rotation=(X=-0.0136,Y=-0.0861,Z=-0.1558,W=0.9839),Translation=(X=0,Y=0,Z=0),Scale3D=(X=1,Y=1,Z=1))))
    WeightMinimum=0
    Weight=@n2.Value
   )
   RigVMFunctionReturnNode(ExecuteContext=@RigVMFunction_Sequence.B)
  ]}
 New Function_ContainedGraph {nodes [
   n1 = RigVMFunctionEntryNode()
   RigVMFunction_Sequence = &RigVMFunction_Sequence(ExecuteContext=@n1.ExecuteContext)
   Delta rotation = RigVMFunctionReferenceNode(Bone Name=calf_l, Maximum=127.5, Minimum=100, ExecuteContext=@RigVMFunction_Sequence.A)
   > &RigUnit_ModifyTransforms(
    ItemToModify="((Item=(Type=Bone,Name=\"dyn_boot_l_helper\"),Transform=(Rotation=(X=0,Y=0,Z=0,W=1),Translation=(X=7.0368,Y=0,Z=0),Scale3D=(X=1,Y=1,Z=1))),(Item=(Type=Bone,Name=\"dyn_pants_l_bck\"),Transform=(Rotation=(X=0.2604,Y=-0.8259,Z=-0.1504,W=0.4769),Translation=(X=-5. ...(truncated)"
    WeightMinimum=0
    Weight=@"Delta rotation".Delta
   )
   > Delta rotation_1 = RigVMFunctionReferenceNode(Bone Name=calf_r, Maximum=127.5, Minimum=100)
   > &RigUnit_ModifyTransforms(
    ItemToModify="((Item=(Type=Bone,Name=\"dyn_pants_r_bck\"),Transform=(Rotation=(X=-0.0436,Y=0.1735,Z=0,W=0.9839),Translation=(X=1.1389,Y=0.1462,Z=-7.6018),Scale3D=(X=1,Y=1,Z=1))),(Item=(Type=Bone,Name=\"dyn_pants_r_inner_01\"),Transform=(Rotation=(X=0,Y=0,Z=-0.1305,W=0.9914) ...(truncated)"
    WeightMinimum=0
    Weight=@"Delta rotation_1".Delta
   )
   > &RigUnit_ParentConstraint(
    AdvancedSettings=(InterpolationType=Average,RotationOrderForFilter=XZY)
    Child=(Type=Bone,Name="dyn_kneecap_r")
    Parents=((Item=(Type=Bone,Name="thigh_r"),Weight=1),(Item=(Type=Bone,Name="calf_r"),Weight=1))
    Weight=1
   )
   > &RigUnit_ParentConstraint(
    AdvancedSettings=(InterpolationType=Average,RotationOrderForFilter=XZY)
    Child=(Type=Bone,Name="dyn_kneecap_l")
    Parents=((Item=(Type=Bone,Name="thigh_l"),Weight=1),(Item=(Type=Bone,Name="calf_l"),Weight=1))
    Weight=1
   )
   > Delta rotation_2 = RigVMFunctionReferenceNode(Bone Name=calf_l, Maximum=100, Minimum=0)
   > &RigUnit_ModifyTransforms(
    ItemToModify=((Item=(Type=Bone,Name="dyn_kneecap_l"),Transform=(Rotation=(X=0,Y=0,Z=0,W=1),Translation=(X=0,Y=0,Z=-1.0543),Scale3D=(X=1,Y=1,Z=1))))
    WeightMinimum=0
    Weight=@"Delta rotation_2".Delta
   )
   > Delta rotation_2_1 = RigVMFunctionReferenceNode(Bone Name=calf_r, Maximum=100, Minimum=0)
   > &RigUnit_ModifyTransforms(
    ItemToModify=((Item=(Type=Bone,Name="dyn_kneecap_r"),Transform=(Rotation=(X=0,Y=0,Z=0,W=1),Translation=(X=-0.7,Y=0,Z=-1.0543),Scale3D=(X=1,Y=1,Z=1))))
    WeightMinimum=0
    Weight=@"Delta rotation_2_1".Delta
   )
   RigVMFunctionReturnNode(ExecuteContext=@RigVMFunction_Sequence.B)
  ]}
 New Function_ContainedGraph {nodes [
   n1 = RigVMFunctionEntryNode()
   RigVMFunction_MathQuaternionToEuler = &RigVMFunction_MathQuaternionToEuler(
    RotationOrder=ZYX
    Value=&RigVMFunction_MathTransformMakeRelative(
     Global=&RigUnit_GetTransform(Item=(Type=Bone,Name="calf_l"), Space=LocalSpace, Item.Name=@n1.Bone Name).Transform
     Parent=&RigUnit_GetTransform(Item=(Type=Bone,Name="calf_l"), Space=LocalSpace, bInitial=true, Item.Name=@n1.Bone Name).Transform
    ).Local.Rotation
   )
   > RigVMFunctionReturnNode(Delta=&RigVMFunction_MathFloatRemap(bClamp=true, SourceMaximum=@n1.Maximum, SourceMinimum=@n1.Minimum, Value=@RigVMFunction_MathQuaternionToEuler.Result.Z).Result)
  ]}
 RigVMFunctionLibrary {nodes [
   Delta Rotation = RigVMCollapseNode(ExecuteContext=())
   Hip Area = RigVMCollapseNode(ExecuteContext=())
   Knee Area = RigVMCollapseNode(ExecuteContext=())
  ]}
 RigVMModel {nodes [
   RigUnit_BeginExecution = &RigUnit_BeginExecution()
   n1 = &RigVMFunction_Sequence(ExecuteContext=@RigUnit_BeginExecution.ExecutePin)
   FN_DeformRig_FNC = RigVMFunctionReferenceNode(
    DeformRig_ Curve="(deformRig_l_upperarm_up_alpha_72_#1=False,deformRig_r_upperarm_up_alpha_73_#2=False,deformRig_l_upperarm_fr_alpha_74_#3=False,deformRig_r_upperarm_fr_alpha_75_CB51B8 ...(truncated)"
    DeformRig_Bool="(upperarm_1_#4=false,elbow_3_#5=false,wrist_7_#6=false,groin_8_#7=True,glute_17_#8=True,knee_9_40C408324CA ...(truncated)"
    DeformRig_Debug=(upperarm_debug_18_#4=False,wrist_debug_20_#6=False,lat_debug_26_#9=false,groinGlute_debug_27_#7=False)
    DeformRig_Value="(upperarm_up_18_#10=(Rotation=(X=0,Y=0,Z=0,W=1),Translation=(X=0,Y=0,Z=8),Scale3D=(X=0.7,Y=1,Z=1)),upperarm_fwd_fr_19_#11=(Rotation=(X=0,Y=0,Z=0,W=1),Translation=(X=0,Y=10,Z=0),Scale3D=(X=0.5,Y=1,Z= ...(truncated)"
    ExecuteContext=@n1.A
   )
   Hip Area = RigVMFunctionReferenceNode(ExecuteContext=@n1.B)
   > Knee Area = RigVMFunctionReferenceNode()
  ]}
]
hierarchy [
 root {loc=<0 0 0>, rot=<0 0 0 1>}
 pelvis {parent=root, loc=<0 -0.7071 93.1191>, rot=<0.0127 0.707 -0.0127 -0.707>}
 spine_01 {parent=pelvis, loc=<6.7494 -0.6388 0>, rot=<0 0 0.034 -0.9994>}
 spine_02 {parent=spine_01, loc=<7.2518 -0.035 0>, rot=<0 0 0.0238 -0.9997>}
 spine_03 {parent=spine_02, loc=<6.0104 0 0>, rot=<0 0 0.0149 -0.9999>}
 spine_04 {parent=spine_03, loc=<8.1024 0 0>, rot=<0 0 -0.0132 -0.9999>}
 spine_05 {parent=spine_04, loc=<10.0146 0 0>, rot=<0 0 -0.0015 -1>}
 attach_cape {parent=spine_05, loc=<9.017 -11.714 0>, rot=<0.0616 -0.7044 -0.0616 -0.7044>}
 neck_01 {parent=spine_05, loc=<11.3138 0.1529 0>, rot=<0 0 -0.2003 -0.9797>}
 neck_02 {parent=neck_01, loc=<4.5627 0 0>, rot=<0 0 0.0504 -0.9987>}
 head {parent=neck_02, loc=<5.3914 0 0>, rot=<0 0 0.115 -0.9934>}
 attach_backpack {parent=spine_05, loc=<-0.1229 -13.4729 0>, rot=<-0.0093 -0.707 0.0093 -0.707>}
 clavicle_l {parent=spine_05, loc=<4.3575 -0.1466 -2.3223>, rot=<0.0465 -0.7289 -0.009 -0.6829>}
 upperarm_l {parent=clavicle_l, loc=<15.5111 0 0>, rot=<0.0226 -0.3234 -0.0088 -0.9459>}
 lowerarm_l {parent=upperarm_l, loc=<26.6514 0 0>, rot=<0.0391 -0.0099 -0.2453 -0.9686>}
 hand_l {parent=lowerarm_l, loc=<22.3491 0 0>, rot=<-0.7081 -0.0507 0.0975 -0.6975>}
 thumb_01_l {parent=hand_l, loc=<1.3118 -0.3191 -2.459>, rot=<0.6546 -0.4042 0.0664 -0.6354>}
 thumb_02_l {parent=thumb_01_l, loc=<5.9336 -0.0448 0.0401>, rot=<-0.0037 0.0213 0.298 -0.9543>}
 thumb_03_l {parent=thumb_02_l, loc=<2.2223 0.3682 -0.0907>, rot=<0.0118 -0.013 0.0066 -0.9998>}
 index_metacarpal_l {parent=hand_l, loc=<1.6535 0.0183 -2.96>, rot=<0.0158 -0.0124 -0.0173 -0.9996>}
 index_01_l {parent=index_metacarpal_l, loc=<8.3008 0.0357 -0.0524>, rot=<-0.0044 0.0485 0.0924 -0.9945>}
 index_02_l {parent=index_01_l, loc=<3.3587 -0.3004 0.0452>, rot=<-0.002 0.0027 0.3225 -0.9466>}
 index_03_l {parent=index_02_l, loc=<2.5578 -0.123 -0.0252>, rot=<-0.002 0.0067 0.2558 -0.9667>}
 middle_metacarpal_l {parent=hand_l, loc=<2.1856 -0.1725 -0.5984>, rot=<0.006 0.0168 -0.0329 -0.9993>}
 middle_01_l {parent=middle_metacarpal_l, loc=<7.3275 -0.2447 0.0129>, rot=<-0.0208 0.0042 0.0858 -0.9961>}
 middle_02_l {parent=middle_01_l, loc=<4.0123 0.1401 0.1175>, rot=<-0.0218 0.0258 0.3512 -0.9357>}
 middle_03_l {parent=middle_02_l, loc=<2.6077 0.0534 0.1373>, rot=<0.0082 0.0188 0.1708 -0.9851>}
 ring_metacarpal_l {parent=hand_l, loc=<2.2074 -0.3625 1.3638>, rot=<-0.0284 0.0722 -0.0458 -0.9959>}
 ring_01_l {parent=ring_metacarpal_l, loc=<7.0295 -0.6197 -0.1446>, rot=<-0.0489 -0.0252 0.1892 -0.9804>}
 ring_02_l {parent=ring_01_l, loc=<3.6174 0.4647 -0.1027>, rot=<0 0 0.2806 -0.9598>}
 ring_03_l {parent=ring_02_l, loc=<2.4152 0.114 0.0199>, rot=<0 0 0.1189 -0.9929>}
 pinky_metacarpal_l {parent=hand_l, loc=<1.7616 -0.9753 3.3814>, rot=<-0.0397 0.0814 -0.069 -0.9935>}
 pinky_01_l {parent=pinky_metacarpal_l, loc=<7.1709 -0.4831 -0.1094>, rot=<-0.0654 -0.0035 0.2242 -0.9723>}
 pinky_02_l {parent=pinky_01_l, loc=<2.8274 0.0893 -0.1043>, rot=<0.0037 0.0011 0.2854 -0.9584>}
 pinky_03_l {parent=pinky_02_l, loc=<2.2046 0.0307 -0.0078>, rot=<-0.0002 0.0098 0.1445 -0.9895>}
 dyn_ini_sleeveCuff_l_B {parent=hand_l, loc=<-0.9225 -0.563 3.7218>, rot=<0.1417 0.1615 -0.054 -0.9752>}
 dyn_sleeveCuff_l_B {parent=dyn_ini_sleeveCuff_l_B, loc=<2.9753 0.0003 0.0004>, rot=<0 0 0 1>}
 dyn_ini_sleeveCuff_C {parent=hand_l, loc=<-0.7917 -5.1361 -2.4788>, rot=<-0.7039 0.2191 0.1436 -0.6603>}
 dyn_sleeveCuff_C {parent=dyn_ini_sleeveCuff_C, loc=<3.3403 -0.0003 0.0002>, rot=<0 0 0 1>}
 dyn_ini_sleeveCuff_l_A {parent=hand_l, loc=<-1.44 2.3956 -0.5373>, rot=<0.7203 0.0855 -0.1569 -0.6703>}
 dyn_sleeveCuff_l_A {parent=dyn_ini_sleeveCuff_l_A, loc=<1.5372 0 0>, rot=<0 0 0 1>}
 lowerarm_twist_02_l {parent=lowerarm_l, loc=<9.6199 0 0>, rot=<0 0 0 -1>}
 dyn_L_sleeve_dynamics {parent=lowerarm_twist_02_l, loc=<-0.867 0 -2.4743>, rot=<0 -0.7071 0 -0.7071>}
 lowerarm_twist_01_l {parent=lowerarm_l, loc=<17.8549 0 0>, rot=<0 0 0 -1>}
 dyn_shirtArm_l_C {parent=lowerarm_l, loc=<15.2208 -0.1946 5.3318>, rot=<0 0 1 0>}
 dyn_shirtArm_l_B {parent=lowerarm_l, loc=<6.7823 1.3166 2.5034>, rot=<0.0871 0.0023 -0.0261 -0.9959>}
 dyn_sleeveBell_l_fr {parent=lowerarm_l, loc=<19.2067 5.0594 -0.1278>, rot=<0.5366 0.3109 -0.2884 -0.7296>}
 dyn_sleeveBell_l_side {parent=lowerarm_l, loc=<17.0646 -5.0572 2.1489>, rot=<-0.3439 0.2541 0.188 -0.8842>}
 dyn_sleeveBell_l_bck {parent=lowerarm_l, loc=<17.2401 -0.9592 -7.2927>, rot=<0.8621 -0.0365 -0.4996 0.0763>}
 upperarm_twist_01_l {parent=upperarm_l, loc=<0.0041 0 0>, rot=<0 0 0 1>}
 upperarm_twist_02_l {parent=upperarm_l, loc=<13.7941 0 0>, rot=<0 0 0 1>}
 dyn_shirtArm_l_A {parent=upperarm_l, loc=<25.0575 0.8416 4.119>, rot=<0.0038 0.0871 -0.0435 -0.9952>}
 clavicle_r {parent=spine_05, loc=<4.3575 -0.1466 2.3223>, rot=<0.7289 0.0465 -0.6829 0.009>}
 upperarm_r {parent=clavicle_r, loc=<-15.5111 0 0>, rot=<0.0226 -0.3234 -0.0088 -0.9459>}
 lowerarm_r {parent=upperarm_r, loc=<-26.6514 0 0>, rot=<0.0391 -0.0099 -0.2453 -0.9686>}
 hand_r {parent=lowerarm_r, loc=<-22.3491 0 0>, rot=<-0.7081 -0.0507 0.0975 -0.6975>}
 thumb_01_r {parent=hand_r, loc=<-1.3118 0.3191 2.459>, rot=<0.6546 -0.4042 0.0664 -0.6354>}
 thumb_02_r {parent=thumb_01_r, loc=<-5.9336 0.0448 -0.0401>, rot=<-0.0037 0.0213 0.298 -0.9543>}
 thumb_03_r {parent=thumb_02_r, loc=<-2.2223 -0.3682 0.0907>, rot=<0.0118 -0.013 0.0066 -0.9998>}
 index_metacarpal_r {parent=hand_r, loc=<-1.6535 -0.0183 2.96>, rot=<0.0158 -0.0124 -0.0173 -0.9996>}
 index_01_r {parent=index_metacarpal_r, loc=<-8.3008 -0.0357 0.0524>, rot=<-0.0044 0.0485 0.0924 -0.9945>}
 index_02_r {parent=index_01_r, loc=<-3.3587 0.3004 -0.0452>, rot=<-0.002 0.0027 0.3225 -0.9466>}
 index_03_r {parent=index_02_r, loc=<-2.5578 0.123 0.0252>, rot=<-0.002 0.0067 0.2558 -0.9667>}
 middle_metacarpal_r {parent=hand_r, loc=<-2.1856 0.1725 0.5984>, rot=<0.006 0.0168 -0.0329 -0.9993>}
 middle_01_r {parent=middle_metacarpal_r, loc=<-7.3275 0.2447 -0.0129>, rot=<-0.0208 0.0042 0.0858 -0.9961>}
 middle_02_r {parent=middle_01_r, loc=<-4.0123 -0.1401 -0.1175>, rot=<-0.0218 0.0258 0.3512 -0.9357>}
 middle_03_r {parent=middle_02_r, loc=<-2.6077 -0.0534 -0.1373>, rot=<0.0082 0.0188 0.1708 -0.9851>}
 ring_metacarpal_r {parent=hand_r, loc=<-2.2074 0.3625 -1.3638>, rot=<-0.0284 0.0722 -0.0458 -0.9959>}
 ring_01_r {parent=ring_metacarpal_r, loc=<-7.0295 0.6197 0.1446>, rot=<-0.0489 -0.0252 0.1892 -0.9804>}
 ring_02_r {parent=ring_01_r, loc=<-3.6174 -0.4647 0.1027>, rot=<0 0 0.2806 -0.9598>}
 ring_03_r {parent=ring_02_r, loc=<-2.4152 -0.114 -0.0199>, rot=<0 0 0.1189 -0.9929>}
 pinky_metacarpal_r {parent=hand_r, loc=<-1.7616 0.9753 -3.3814>, rot=<-0.0397 0.0814 -0.069 -0.9935>}
 pinky_01_r {parent=pinky_metacarpal_r, loc=<-7.1709 0.4831 0.1094>, rot=<-0.0654 -0.0035 0.2242 -0.9723>}
 pinky_02_r {parent=pinky_01_r, loc=<-2.8274 -0.0893 0.1043>, rot=<0.0037 0.0011 0.2854 -0.9584>}
 pinky_03_r {parent=pinky_02_r, loc=<-2.2046 -0.0307 0.0078>, rot=<-0.0002 0.0098 0.1445 -0.9895>}
 dyn_sleeveCuff_r_A {parent=hand_r, loc=<1.154 -0.147 3.62>, rot=<0.2303 -0.2335 0.9443 0.0264>}
 dyn_ini_sleeveCuff_r_B {parent=hand_r, loc=<0.4689 -1.017 -2.9356>, rot=<0.0342 -0.9374 0.2217 -0.2663>}
 dyn_sleeveCuff_r_B {parent=dyn_ini_sleeveCuff_r_B, loc=<2.6119 0 0>, rot=<-0.0398 -0.2581 -0.0106 -0.9652>}
 dyn_ini_sleeveCuff_r_C {parent=hand_r, loc=<1.1403 4.9896 -0.706>, rot=<-0.0667 -0.8043 -0.5829 -0.0942>}
 dyn_sleeveCuff_r_C {parent=dyn_ini_sleeveCuff_r_C, loc=<1.5826 0 0>, rot=<0.0356 0.2401 -0.0088 -0.9701>}
 sleeve_corrective_r {parent=hand_r, loc=<3.5485 0.652 0.1083>, rot=<0 0 0 1>}
 lowerarm_twist_02_r {parent=lowerarm_r, loc=<-9.6199 0 0>, rot=<0 0 0 -1>}
 dyn_R_sleeve_dynamics {parent=lowerarm_twist_02_r, loc=<0.8668 0.0001 2.4745>, rot=<0 -0.7071 0 -0.7071>}
 lowerarm_twist_01_r {parent=lowerarm_r, loc=<-17.8549 0 0>, rot=<0 0 0 -1>}
 dyn_shirtArm_r_F {parent=lowerarm_r, loc=<-12.7882 -7.6887 -0.0541>, rot=<0.8607 0.0698 0.1052 0.4931>}
 dyn_shirtArm_r_E {parent=lowerarm_r, loc=<-6.786 2.7573 -4.2819>, rot=<0.9664 -0.1305 -0.0013 -0.2216>}
 dyn_sleeveBell_r_bck {parent=lowerarm_r, loc=<-17.4651 1.0423 7.5585>, rot=<0.6281 0.0583 0.7732 0.0652>}
 dyn_sleeveBell_r_side {parent=lowerarm_r, loc=<-17.0643 5.057 -2.1486>, rot=<-0.3722 -0.6794 -0.2566 -0.578>}
 dyn_sleeveBell_r_fr {parent=lowerarm_r, loc=<-19.207 -5.0592 0.1275>, rot=<0.327 -0.7469 0.3778 -0.4387>}
 upperarm_twist_01_r {parent=upperarm_r, loc=<-0.0041 0 0>, rot=<0 0 0 -1>}
 upperarm_twist_02_r {parent=upperarm_r, loc=<-13.7941 0 0>, rot=<0 0 0 -1>}
 dyn_shirtArm_r_D {parent=upperarm_r, loc=<-13.5336 4.36 5.3346>, rot=<-0.246 -0.7077 -0.5939 0.2932>}
 dyn_shirtArm_r_C {parent=upperarm_r, loc=<-0.3985 0.3523 -3.5097>, rot=<-0.0396 -0.9941 -0.0472 -0.0889>}
 dyn_shirtArm_r_B {parent=clavicle_r, loc=<-11.6635 -3.4187 5.5459>, rot=<-0.1582 0.7756 -0.0793 -0.6059>}
 dyn_shirtArm_r_A {parent=clavicle_r, loc=<-12.1241 6.2427 1.7203>, rot=<-0.4564 -0.6218 -0.3961 0.4981>}
 dyn_shirtHole_r_fr {parent=spine_03, loc=<3.3421 9.0462 10.2669>, rot=<-0.0211 0.4974 0.8653 -0.0579>}
 dyn_shirtHole_r_bck {parent=spine_02, loc=<6.1917 -9.0449 6.0458>, rot=<0.0542 -0.7753 0.6278 0.0439>}
 dyn_shirtHole_bot_r_fr {parent=spine_01, loc=<3.036 5.6888 11.4022>, rot=<0.4686 0.0239 0.8821 -0.0421>}
 dyn_shirtHole_bot_r_side {parent=spine_01, loc=<4.3381 -3.7682 12.4839>, rot=<0.4261 -0.0383 0.9035 -0.0246>}
 dyn_shirtHole_bot_r_bck {parent=spine_01, loc=<3.622 -7.9702 8.591>, rot=<0.2993 -0.3157 0.8992 -0.0472>}
 dyn_shirtHole_l {parent=spine_01, loc=<4.5164 2.4878 -11.7799>, rot=<-0.0503 -0.7053 0.0503 -0.7053>}
 thigh_l {parent=pelvis, loc=<-2.5693 -0.0109 -7.7501>, rot=<0.0012 0.0572 0.0155 -0.9982>}
 calf_l {parent=thigh_l, loc=<-39.0484 0 0>, rot=<0.0006 -0.0161 -0.0391 -0.9991>}
 foot_l {parent=calf_l, loc=<-42.561 0.1077 0.4292>, rot=<0.0418 -0.0455 0.0343 -0.9975>}
 ball_l {parent=foot_l, loc=<-6.1599 13.8264 0.6806>, rot=<0.0001 -0.0001 -0.7063 -0.7079>}
 calf_twist_02_l {parent=calf_l, loc=<-18.8064 -0.0967 0.0397>, rot=<0.0237 0.0095 -0.0035 -0.9997>}
 base_boot_l_helper {parent=calf_twist_02_l, loc=<7.6857 -8.7393 -0.5847>, rot=<0.0383 -0.9985 -0.028 0.0278>}
 dyn_boot_l_helper {parent=base_boot_l_helper, loc=<0 0 0>, rot=<0 0 0 1>}
 calf_twist_01_l {parent=calf_l, loc=<-35.813 -0.0967 0.0397>, rot=<0.0237 0.0095 -0.0035 -0.9997>}
 thigh_twist_01_l {parent=thigh_l, loc=<-21.6303 0 0>, rot=<0.0013 0 0 -1>}
 dyn_kneecap_l {parent=thigh_l, loc=<-37.0758 4.3261 -0.0067>, rot=<0.4305 0.561 0.561 0.4305>}
 dyn_pants_l_inner {parent=thigh_l, loc=<-34.2091 2.841 8.0262>, rot=<0.0085 0.2327 0.9723 -0.0183>}
 dyn_pants_l_bck {parent=thigh_l, loc=<-33.029 -8.1731 -0.2469>, rot=<0.1036 -0.7757 0.6075 -0.1365>}
 dyn_pants_l_outer {parent=thigh_l, loc=<-33.5937 -2.0873 -6.907>, rot=<0.0038 -0.9943 0.0608 -0.0871>}
 dyn_legCloth_base {parent=thigh_l, loc=<0.4346 5.8018 -3.5371>, rot=<-0.0589 -0.9133 -0.3979 -0.0646>}
 dyn_legCloth_bck_01 {parent=dyn_legCloth_base, loc=<0.9589 -0.3135 -0.4842>, rot=<0.084 -0.034 0.0142 -0.9958>}
 dyn_legCloth_bck_02 {parent=dyn_legCloth_bck_01, loc=<10.7611 0 0>, rot=<-0.0309 -0.0178 0.0553 -0.9978>}
 dyn_legCloth_bck_03 {parent=dyn_legCloth_bck_02, loc=<10.2023 0 0>, rot=<-0.0436 0 0 -0.999>}
 dyn_legCloth_fr_01 {parent=dyn_legCloth_base, loc=<1.1794 0.0229 0.4614>, rot=<0.0562 -0.0102 0.0276 -0.998>}
 dyn_legCloth_fr_02 {parent=dyn_legCloth_fr_01, loc=<10.8845 0 0>, rot=<0.0075 -0.032 0.0042 -0.9995>}
 dyn_legCloth_fr_03 {parent=dyn_legCloth_fr_02, loc=<10.1008 0 0>, rot=<-0.0174 -0.0513 -0.003 -0.9985>}
 dyn_legCloth_fr_04 {parent=dyn_legCloth_fr_03, loc=<7.204 0.2516 0>, rot=<0.0162 -0.0088 0.0281 -0.9994>}
 dyn_legCloth_fr_05 {parent=dyn_legCloth_fr_04, loc=<7.4543 0 0>, rot=<0.0182 -0.0389 -0.0166 -0.9989>}
 deform_groin_base_l {parent=thigh_l, loc=<-2.1908 8.1142 -1.7442>, rot=<0 0 0 -1>}
 deform_groin_l {parent=deform_groin_base_l, loc=<0 0 0>, rot=<0 0 0 1>}
 base_pants_helper_l {parent=thigh_l, loc=<-21.6303 0 0>, rot=<0.0013 0 0 -1>}
 dyn_pants_helper_l {parent=base_pants_helper_l, loc=<0 0 0>, rot=<0 0 0 1>}
 thigh_r {parent=pelvis, loc=<-2.5693 -0.0109 7.7501>, rot=<0.0572 -0.0012 0.9982 0.0155>}
 calf_r {parent=thigh_r, loc=<39.0484 0 0>, rot=<0.0006 -0.0161 -0.0391 -0.9991>}
 foot_r {parent=calf_r, loc=<42.561 -0.1077 -0.4292>, rot=<0.0418 -0.0455 0.0343 -0.9975>}
 ball_r {parent=foot_r, loc=<6.1599 -13.8264 -0.6806>, rot=<0.0001 -0.0001 -0.7063 -0.7079>}
 calf_twist_02_r {parent=calf_r, loc=<18.8064 0.0967 -0.0397>, rot=<0.0237 0.0095 -0.0035 -0.9997>}
 calf_twist_01_r {parent=calf_r, loc=<35.813 0.0967 -0.0397>, rot=<0.0237 0.0095 -0.0035 -0.9997>}
 thigh_twist_01_r {parent=thigh_r, loc=<21.6303 0 0>, rot=<0.0013 0 0 -1>}
 dyn_kneecap_r {parent=thigh_r, loc=<37.0757 -4.3261 0.0064>, rot=<-0.561 0.4305 0.4305 -0.561>}
 dyn_pants_r_inner_01 {parent=thigh_r, loc=<21.8223 1.8745 -8.4854>, rot=<0.9652 -0.0118 -0.0347 -0.2588>}
 dyn_pants_r_inner_02 {parent=dyn_pants_r_inner_01, loc=<9.2542 0 0>, rot=<0 0 0 1>}
 dyn_pants_r_bck {parent=thigh_r, loc=<19.3276 7.2863 2.341>, rot=<0.5628 0.0262 -0.1138 -0.8183>}
 dyn_pants_r_outer {parent=thigh_r, loc=<19.7642 -2.7591 5.2942>, rot=<-0.1643 0.0945 -0.0158 -0.9817>}
 dyn_clothscrapLeg_r {parent=thigh_r, loc=<5.4457 -1.1108 6.237>, rot=<-0.1204 0.1894 -0.0191 -0.9743>}
 dyn_belt_01 {parent=thigh_r, loc=<2.1341 -9.158 -4.0563>, rot=<-0.6846 -0.0378 0.1213 -0.7178>}
 dyn_belt_02 {parent=dyn_belt_01, loc=<7.3245 0 0>, rot=<-0.0174 -0.0003 0.0174 -0.9997>}
 deform_groin_base_r {parent=thigh_r, loc=<2.1908 -8.1142 1.7442>, rot=<0 0 0 -1>}
 deform_groin_r {parent=deform_groin_base_r, loc=<0 0 0>, rot=<0 0 0 1>}
 base_pants_helper_r {parent=thigh_r, loc=<21.6285 0.0009 -0.0682>, rot=<0.0489 0.0057 -0.0008 -0.9988>}
 dyn_pants_helper_r {parent=base_pants_helper_r, loc=<0 0 0>, rot=<0 0 0 1>}
 dyn_crossbodyCloth_base {parent=pelvis, loc=<10.4726 7.3775 -10.5174>, rot=<-0.0665 -0.9098 -0.3769 -0.1604>}
 dyn_crossbodyCloth_01 {parent=dyn_crossbodyCloth_base, loc=<0 0 0>, rot=<0.0302 -0.0149 0.0037 -0.9994>}
 dyn_crossbodyCloth_02 {parent=dyn_crossbodyCloth_01, loc=<6.2253 0 0>, rot=<0.0523 0 0 -0.9986>}
 dyn_crossbodyStrap_base {parent=pelvis, loc=<5.0809 -3.2144 -14.1172>, rot=<-0.2834 -0.9266 0.2072 -0.1346>}
 dyn_crossbodyStrap_01 {parent=dyn_crossbodyStrap_base, loc=<0 0 0>, rot=<0 0 0 1>}
 dyn_crossbodyStrap_02 {parent=dyn_crossbodyStrap_01, loc=<5.8446 0 0>, rot=<0.0872 0 0 -0.9962>}
 deform_glute_base_l {parent=pelvis, loc=<-8.8064 -8.8222 -8.4287>, rot=<0.0275 0.0501 -0.4472 -0.8926>}
 deform_glute_l {parent=deform_glute_base_l, loc=<0 0 0>, rot=<0 0 0 1>}
 deform_glute_base_r {parent=pelvis, loc=<-8.8062 -8.8222 8.4287>, rot=<0.0501 -0.0275 0.8926 -0.4472>}
 deform_glute_r {parent=deform_glute_base_r, loc=<0 0 0>, rot=<0 0 0 1>}
]
