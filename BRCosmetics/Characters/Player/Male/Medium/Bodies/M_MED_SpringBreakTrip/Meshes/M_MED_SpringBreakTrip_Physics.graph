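# Physics asset for the male medium SpringBreakTrip body cosmetic, defining capsule bodies and joints for its dangling crossbody cloth, crossbody strap and leg cloth.
# Capsules on pelvis, lower spine, both thighs and left calf form the body core.
# Crossbody chains root at the pelvis, while front and back leg cloth chains share a base on the left thigh.
# Chain links use tight swing/twist limits, except free-moving root joints and first front leg link.
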PhysicsAsset /BRCosmetics/Characters/Player/Male/Medium/Bodies/M_MED_SpringBreakTrip/Meshes/M_MED_SpringBreakTrip_Physics
bodies {
 calf_l [{shapeType=Capsule, length=31.0273, radius=7.4727, shapeName=calf_l_capsule, center=<-22.2319 0.6477 0>, rotation=<rot 90 0 0.0002>}]
 dyn_crossbodyCloth_01 [{shapeType=Capsule, length=5, radius=3, shapeName=dyn_crossbodyCloth_01_capsule, center=<2.4412 0 2.809>, rotation=<rot 0.0001 0.0001 90>}]
 dyn_crossbodyCloth_02 [{
   shapeType=Capsule
   length=5.0466
   radius=3.9534
   shapeName=dyn_crossbodyCloth_02_capsule
   center=<3.4868 0 2.0023>
   rotation=<rot 0.0001 0.0001 90>
  }]
 dyn_crossbodyCloth_base [{shapeType=Capsule, length=1, radius=1, shapeName=dyn_crossbodyCloth_base_capsule}]
 dyn_crossbodyStrap_01 [{
   shapeType=Capsule
   length=4.4108
   radius=1.1671
   shapeName=dyn_crossbodyStrap_01_capsule
   center=<5.5011 0.2225 0.526>
   rotation=<rot 71.793 86.8027 79.4479>
  }]
 dyn_crossbodyStrap_02 [{
   shapeType=Capsule
   length=8.4014
   radius=1.3792
   shapeName=dyn_crossbodyStrap_02_capsule
   center=<7.0385 -0.7661 -0.7887>
   rotation=<rot 82.6011 91.9386 -11.5343>
  }]
 dyn_crossbodyStrap_base [{shapeType=Capsule, length=1, radius=1, shapeName=dyn_crossbodyStrap_base_capsule}]
 dyn_legCloth_base [{shapeType=Capsule, length=1, radius=1, shapeName=dyn_legCloth_base_capsule}]
 dyn_legCloth_bck_01 [{
   shapeType=Capsule
   length=8.7718
   radius=2.2282
   shapeName=dyn_legCloth_bck_01_capsule
   center=<5.45 0 1.8863>
   rotation=<rot -90 180 87.7321>
  }]
 dyn_legCloth_bck_02 [{
   shapeType=Capsule
   length=8.1313
   radius=2.8687
   shapeName=dyn_legCloth_bck_02_capsule
   center=<5.45 0 2.1024>
   rotation=<rot -90 180 94.5828>
  }]
 dyn_legCloth_bck_03 [
  {
   shapeType=Capsule
   length=8.7362
   radius=3.282
   shapeName=dyn_legCloth_bck_03_capsule
   center=<5.3621 0.8711 2.3602>
   rotation=<rot -90 180 95.7585>
  }
  {
   shapeType=Capsule
   length=11.4582
   radius=3.282
   shapeName=dyn_legCloth_bck_03_capsule
   center=<17.1259 0.9645 -0.8152>
   rotation=<rot -82.6539 -174.2122 -93.3018>
  }
 ]
 dyn_legCloth_fr_01 [{shapeType=Capsule, length=5.5, radius=5, shapeName=dyn_legCloth_fr_01_capsule, center=<3.9725 0 0>, rotation=<rot 90 -164.9778 -164.978>}]
 dyn_legCloth_fr_02 [{
   shapeType=Capsule
   length=8.5657
   radius=2.4543
   shapeName=dyn_legCloth_fr_02_capsule
   center=<3.3393 0 0.5221>
   rotation=<rot 90 175.2364 -95.0676>
  }]
 dyn_legCloth_fr_03 [{
   shapeType=Capsule
   length=8.5657
   radius=1.9343
   shapeName=dyn_legCloth_fr_03_capsule
   center=<-0.2158 0 0.8349>
   rotation=<rot 90 -165.9638 -75.9639>
  }]
 dyn_legCloth_fr_04 [{
   shapeType=Capsule
   length=6.551
   radius=2.7126
   shapeName=dyn_legCloth_fr_04_capsule
   center=<0.8431 0 1.6968>
   rotation=<rot 90 -165.9638 101.6544>
  }]
 dyn_legCloth_fr_05 [{
   shapeType=Capsule
   length=10.1157
   radius=3.3569
   shapeName=dyn_legCloth_fr_05_capsule
   center=<4.629 -0.3171 2.6673>
   rotation=<rot 85.0172 -179.9884 90.1009>
  }]
 pelvis [{shapeType=Capsule, length=20, radius=7, shapeName=pelvis_capsule}]
 spine_01 [{shapeType=Capsule, length=20, radius=7, shapeName=spine_01_capsule}]
 thigh_l [{
   shapeType=Capsule
   length=26.4177
   radius=8.5189
   shapeName=thigh_l_capsule
   center=<-21.5364 0.1473 1.2649>
   rotation=<rot 90 169.958 174.9585>
  }]
 thigh_r [{shapeType=Capsule, length=25, radius=12, shapeName=thigh_r_capsule, center=<19.1173 -0.0001 0>, rotation=<rot 90 179.9999 0.0002>}]
}
constraints [
 {bone1Name=dyn_crossbodyCloth_01, bone2Name=dyn_crossbodyCloth_base, swing1LimitDegrees=15, swing2LimitDegrees=25, twistLimitDegrees=0}
 {bone1Name=dyn_crossbodyCloth_02, bone2Name=dyn_crossbodyCloth_01, swing1LimitDegrees=15, swing2LimitDegrees=25, twistLimitDegrees=0}
 {bone1Name=dyn_crossbodyStrap_01, bone2Name=dyn_crossbodyStrap_base, swing1LimitDegrees=10, swing2LimitDegrees=10, twistLimitDegrees=0}
 {bone1Name=dyn_crossbodyStrap_02, bone2Name=dyn_crossbodyStrap_01, swing1LimitDegrees=10, swing2LimitDegrees=20, twistLimitDegrees=0}
 {bone1Name=dyn_legCloth_bck_01, bone2Name=dyn_legCloth_base, swing1LimitDegrees=15, swing2LimitDegrees=10, twistLimitDegrees=10}
 {bone1Name=dyn_legCloth_bck_02, bone2Name=dyn_legCloth_bck_01, swing1LimitDegrees=15, swing2LimitDegrees=10, twistLimitDegrees=10}
 {bone1Name=dyn_legCloth_bck_03, bone2Name=dyn_legCloth_bck_02, swing1LimitDegrees=15, swing2LimitDegrees=10, twistLimitDegrees=10}
 {
  bone1Name=dyn_legCloth_base
  bone2Name=thigh_l
  swing1LimitDegrees=45
  swing1Motion=Free
  swing2LimitDegrees=45
  swing2Motion=Free
  twistLimitDegrees=45
  twistMotion=Free
 }
 {
  bone1Name=calf_l
  bone2Name=thigh_l
  swing1LimitDegrees=45
  swing1Motion=Free
  swing2LimitDegrees=45
  swing2Motion=Free
  twistLimitDegrees=45
  twistMotion=Free
 }
 {
  bone1Name=thigh_l
  bone2Name=pelvis
  swing1LimitDegrees=45
  swing1Motion=Free
  swing2LimitDegrees=45
  swing2Motion=Free
  twistLimitDegrees=45
  twistMotion=Free
 }
 {
  bone1Name=thigh_r
  bone2Name=pelvis
  swing1LimitDegrees=45
  swing1Motion=Free
  swing2LimitDegrees=45
  swing2Motion=Free
  twistLimitDegrees=45
  twistMotion=Free
 }
 {
  bone1Name=dyn_crossbodyCloth_base
  bone2Name=pelvis
  swing1LimitDegrees=45
  swing1Motion=Free
  swing2LimitDegrees=45
  swing2Motion=Free
  twistLimitDegrees=45
  twistMotion=Free
 }
 {
  bone1Name=dyn_crossbodyStrap_base
  bone2Name=pelvis
  swing1LimitDegrees=45
  swing1Motion=Free
  swing2LimitDegrees=45
  swing2Motion=Free
  twistLimitDegrees=45
  twistMotion=Free
 }
 {
  bone1Name=spine_01
  bone2Name=pelvis
  swing1LimitDegrees=45
  swing1Motion=Free
  swing2LimitDegrees=45
  swing2Motion=Free
  twistLimitDegrees=45
  twistMotion=Free
 }
 {
  bone1Name=dyn_legCloth_fr_01
  bone2Name=dyn_legCloth_base
  swing1LimitDegrees=45
  swing1Motion=Free
  swing2LimitDegrees=45
  swing2Motion=Free
  twistLimitDegrees=45
  twistMotion=Free
 }
 {bone1Name=dyn_legCloth_fr_02, bone2Name=dyn_legCloth_fr_01, swing1LimitDegrees=20, swing2LimitDegrees=20, twistLimitDegrees=10}
 {bone1Name=dyn_legCloth_fr_03, bone2Name=dyn_legCloth_fr_02, swing1LimitDegrees=20, swing2LimitDegrees=20, twistLimitDegrees=10}
 {bone1Name=dyn_legCloth_fr_04, bone2Name=dyn_legCloth_fr_03, swing1LimitDegrees=10, swing2LimitDegrees=10, twistLimitDegrees=10}
 {bone1Name=dyn_legCloth_fr_05, bone2Name=dyn_legCloth_fr_04, swing1LimitDegrees=10, swing2LimitDegrees=10, twistLimitDegrees=10}
]
