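# Control Rig for the SpringBreakTrip male medium face accessory that poses dynamic hair bones, including the back bases, from head and pelvis orientation.
# A DeltaRotation function remaps head-to-neck_01 rotation into "Head Up" and "Head Back Drastic" values, while a spherical pose reader yields "Pelvis Low".
# Head Up weights one hair offset; Head Back Drastic times Pelvis Low weights a second.
# Two head pose readers are left unconnected.
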ControlRigBlueprint /BRCosmetics/Characters/Player/Male/Medium/Bodies/M_MED_SpringBreakTrip/Meshes/Parts/M_MED_SpringBreakTrip_FaceAcc_CtrlRig
graphs [
 New Function_ContainedGraph {nodes [
   n1 = RigVMFunctionEntryNode()
   RigVMFunction_MathQuaternionToEuler = &RigVMFunction_MathQuaternionToEuler(
    RotationOrder=ZYX
    Value=&RigUnit_GetRelativeTransformForItem(
     Child=(Type=Bone,Name="hand_l")
     Parent=(Type=Bone,Name="lowerarm_l")
     bChildInitial=False
     bParentInitial=False
     Child.Name=@n1.Child
     Parent.Name=@n1.Parent
    ).RelativeTransform.Rotation
   )
   Remap = &RigVMFunction_MathDoubleRemap(
    TargetMaximum=1
    TargetMinimum=0
    Value=@RigVMFunction_MathQuaternionToEuler.Result.Z
    SourceMaximum=@n1.Source Maximum
    SourceMinimum=@n1.Source Minimum
   )
   > RigVMFunctionReturnNode(Delta=@Remap.Result)
  ]}
 RigVMFunctionLibrary {nodes [DeltaRotation = RigVMCollapseNode(ExecuteContext=())]}
 RigVMFunction_Sequence_1_ContainedGraph {nodes [
   n1 = RigVMFunctionEntryNode()
   RigVMFunction_Sequence = &RigVMFunction_Sequence(ExecuteContext=@n1.ExecuteContext)
   n2 = &RigVMFunction_Sequence(ExecuteContext=@RigVMFunction_Sequence.B)
   RigVMFunctionReturnNode(A=@RigVMFunction_Sequence.A, B=@n2.A, C=@n2.B)
  ]}
 RigVMModel {nodes [
   RigUnit_BeginExecution = &RigUnit_BeginExecution()
   RigVMFunction_Sequence = RigVMAggregateNode(ExecuteContext=@RigUnit_BeginExecution.ExecutePin)
   DeltaRotation = RigVMFunctionReferenceNode(Child=head, Parent=neck_01, Source Maximum=-37.85, Source Minimum=-17, ExecuteContext=@RigVMFunction_Sequence.A)
   > RigVMVariableNode(Value=@DeltaRotation.Delta) {variable="Head Up "}
   > n1 = &RigUnit_SphericalPoseReader(
    ActiveRegionScaleFactors=(PositiveWidth=0.3,NegativeWidth=0.7,PositiveHeight=1,NegativeHeight=1)
    ActiveRegionSize=0.1
    Debug=(bDrawDebug=true,bDraw2D=False,bDrawLocalAxes=False,DebugScale=25,DebugSegments=20,DebugThickness=0.25)
    DriverAxis=(X=0,Y=1,Z=0)
    DriverItem=(Type=Bone,Name="pelvis")
    FalloffRegionScaleFactors=(PositiveWidth=0.2,NegativeWidth=0.4,PositiveHeight=0.4,NegativeHeight=0.4)
    FalloffSize=0.22
    FlipHeightScaling=False
    FlipWidthScaling=False
    OptionalParentItem=(Type=Bone,Name="None")
    RotationOffset=(X=70,Y=0,Z=18.2708)
   )
   > RigVMVariableNode(Value=@n1.OutputParam) {variable=Pelvis Low}
   > DeltaRotation_1 = RigVMFunctionReferenceNode(Child=head, Parent=neck_01, Source Maximum=-57, Source Minimum=-38)
   > RigVMVariableNode(Value=@DeltaRotation_1.Delta) {variable=Head Back Drastic}
   AlphaInterp = &RigVMFunction_AlphaInterp(
    ClampMax=0.8
    InterpSpeedDecreasing=10
    InterpSpeedIncreasing=10
    Scale=1
    bClampResult=False
    bInterpResult=False
    Value=RigVMVariableNode() {variable="Head Up "}.Value
   )
   &RigUnit_ModifyTransforms(
    ItemToModify="((Item=(Type=Bone,Name=\"dyn_hair_bck_l_base\"),Transform=(Rotation=(X=0,Y=-0.4226,Z=0,W=0.9063),Translation=(X=0,Y=0,Z=0),Scale3D=(X=1,Y=1,Z=1))),(Item=(Type=Bone,Name=\"dyn_hair_bck_r_base\"),Transform=(Rotation=(X=0,Y=-0.342,Z=0,W=0.9397),Translation=(X=0,Y ...(truncated)"
    WeightMinimum=0
    Weight=@AlphaInterp.Result
    ExecutePin=@RigVMFunction_Sequence.B
   )
   n2 = &RigVMFunction_AlphaInterp(
    InterpSpeedDecreasing=25
    InterpSpeedIncreasing=25
    Scale=1
    bClampResult=False
    bInterpResult=False
    Value=&RigVMFunction_MathDoubleMul(A=RigVMVariableNode() {variable=Head Back Drastic}.Value, B=RigVMVariableNode() {variable=Pelvis Low}.Value).Result
   )
   > &RigUnit_ModifyTransforms(
    ItemToModify="((Item=(Type=Bone,Name=\"dyn_hair_bck_l_base\"),Transform=(Rotation=(X=0,Y=-0.2588,Z=0,W=0.9659),Translation=(X=0,Y=0,Z=0),Scale3D=(X=1,Y=1,Z=1))),(Item=(Type=Bone,Name=\"dyn_hair_bck_r_base\"),Transform=(Rotation=(X=-0.0151,Y=-0.3407,Z=0.0858,W=0.9361),Transl ...(truncated)"
    WeightMinimum=0
    Weight=@n2.Result
   )
   &RigUnit_SphericalPoseReader(
    ActiveRegionScaleFactors=(PositiveWidth=2,NegativeWidth=0.6,PositiveHeight=2,NegativeHeight=2)
    ActiveRegionSize=0.1
    Debug=(bDrawDebug=false,bDraw2D=false,bDrawLocalAxes=False,DebugScale=25,DebugSegments=20,DebugThickness=0.25)
    DriverAxis=(X=1,Y=0,Z=0)
    DriverItem=(Type=Bone,Name="head")
    ExecutePin=()
    FalloffRegionScaleFactors=(PositiveWidth=0.7,NegativeWidth=0.5,PositiveHeight=0.7,NegativeHeight=0.7)
    FalloffSize=0.3
    FlipHeightScaling=False
    FlipWidthScaling=False
    OptionalParentItem=(Type=Bone,Name="None")
    RotationOffset=(X=82.1852,Y=16.0762,Z=-120.4624)
   )
   &RigUnit_SphericalPoseReader(
    ActiveRegionScaleFactors=(PositiveWidth=1,NegativeWidth=1,PositiveHeight=1,NegativeHeight=1)
    ActiveRegionSize=0.1
    Debug=(bDrawDebug=false,bDraw2D=False,bDrawLocalAxes=False,DebugScale=25,DebugSegments=20,DebugThickness=0.25)
    DriverAxis=(X=1,Y=0,Z=0)
    DriverItem=(Type=Bone,Name="head")
    ExecutePin=()
    FalloffRegionScaleFactors=(PositiveWidth=0.7,NegativeWidth=0.7,PositiveHeight=0.7,NegativeHeight=0.7)
    FalloffSize=0.22
    FlipHeightScaling=False
    FlipWidthScaling=False
    OptionalParentItem=(Type=Bone,Name="None")
    RotationOffset=(X=94.7838,Y=-26.3245,Z=-142.8814)
   )
  ]}
]
hierarchy [
 root {loc=<0 0 0>, rot=<0 0 0 1>}
 pelvis {parent=root, loc=<0 -0.7071 93.1191>, rot=<0.0127 0.707 -0.0127 -0.707>}
 spine_01 {parent=pelvis, loc=<6.7494 -0.6388 0>, rot=<0 0 0.034 -0.9994>}
 spine_02 {parent=spine_01, loc=<7.2518 -0.035 0>, rot=<0 0 0.0238 -0.9997>}
 spine_03 {parent=spine_02, loc=<6.0104 0 0>, rot=<0 0 0.0149 -0.9999>}
 spine_04 {parent=spine_03, loc=<8.1024 0 0>, rot=<0 0 -0.0132 -0.9999>}
 spine_05 {parent=spine_04, loc=<10.0146 0 0>, rot=<0 0 -0.0015 -1>}
 neck_01 {parent=spine_05, loc=<11.3138 0.1529 0>, rot=<0 0 -0.2003 -0.9797>}
 neck_02 {parent=neck_01, loc=<4.5627 0 0>, rot=<0 0 0.0504 -0.9987>}
 head {parent=neck_02, loc=<5.3914 0 0>, rot=<0 0 0.115 -0.9934>}
 dyn_hair_fr_l_01 {parent=head, loc=<7.253 1.4928 -7.635>, rot=<0.013 -0.9898 -0.1419 0.0001>}
 dyn_hair_fr_l_02 {parent=dyn_hair_fr_l_01, loc=<2.4954 2.3995 0.2099>, rot=<0.5039 0.0551 0.1004 -0.8561>}
 dyn_hair_fr_l_03 {parent=dyn_hair_fr_l_02, loc=<12.9197 0.9151 0.4915>, rot=<0.0377 0.245 -0.1489 -0.9573>}
 dyn_hair_fr_l_04 {parent=dyn_hair_fr_l_03, loc=<8.3186 0 0>, rot=<0 -0.0236 0.0021 -0.9997>}
 dyn_hair_fr_l_05 {parent=dyn_hair_fr_l_04, loc=<6.621 0 0>, rot=<0 0 0 1>}
 dyn_hair_fr_l_06 {parent=dyn_hair_fr_l_01, loc=<6.9192 -2.6079 0.4421>, rot=<0.0658 0.0873 0.0005 -0.994>}
 dyn_hair_fr_l_07 {parent=dyn_hair_fr_l_06, loc=<7.0648 0 0>, rot=<0 0 -0.1305 -0.9914>}
 dyn_hair_fr_r_01 {parent=head, loc=<7.2464 2.2824 7.6349>, rot=<0.0279 0.6546 0.7493 -0.0967>}
 dyn_hair_fr_r_02 {parent=dyn_hair_fr_r_01, loc=<3.9285 0.567 -0.2407>, rot=<0.0814 0.1415 0.0847 -0.9829>}
 dyn_hair_fr_r_03 {parent=dyn_hair_fr_r_02, loc=<11.0729 0 -0.691>, rot=<-0.076 0.1435 0.0292 -0.9863>}
 dyn_hair_fr_r_04 {parent=dyn_hair_fr_r_03, loc=<7.8137 0 0>, rot=<-0.0186 -0.0144 0.0101 -0.9997>}
 dyn_hair_fr_r_05 {parent=dyn_hair_fr_r_04, loc=<7.95 0 0>, rot=<0 0 0 1>}
 dyn_hair_fr_r_06 {parent=dyn_hair_fr_r_02, loc=<9.8942 3.6237 -2.0397>, rot=<0.2961 0.1656 -0.0522 -0.9392>}
 dyn_hair_pigtail_r_01 {parent=head, loc=<16.9339 2.2969 6.3297>, rot=<0.3383 -0.1291 0.9307 -0.0512>}
 dyn_hair_pigtail_r_02 {parent=dyn_hair_pigtail_r_01, loc=<5.57 0 0>, rot=<-0.0165 -0.2067 0.0087 -0.9782>}
 dyn_hair_pigtail_r_03 {parent=dyn_hair_pigtail_r_02, loc=<4.8964 0 0>, rot=<0 -0.0872 0 -0.9962>}
 dyn_hair_ribbon_A_01 {parent=head, loc=<14.2814 3.4227 7.4706>, rot=<0.1348 0.2934 0.9457 0.0382>}
 dyn_hair_ribbon_A_02 {parent=dyn_hair_ribbon_A_01, loc=<7.1405 0 0>, rot=<-0.0205 0.0009 0.0061 -0.9998>}
 dyn_hair_ribbon_A_03 {parent=dyn_hair_ribbon_A_02, loc=<6.9799 0 0>, rot=<0 0 0 1>}
 dyn_hair_ribbon_B_01 {parent=head, loc=<14.2404 4.161 6.4845>, rot=<0.1337 0.4475 0.8835 0.0374>}
 dyn_hair_ribbon_B_02 {parent=dyn_hair_ribbon_B_01, loc=<7.5399 0 0>, rot=<-0.0886 -0.0825 0.0694 -0.9902>}
 dyn_hair_ribbon_B_03 {parent=dyn_hair_ribbon_B_02, loc=<7.6765 0 0>, rot=<0 0 0 1>}
 dyn_hair_beads_l_01 {parent=head, loc=<0.3426 6.0263 -5.7431>, rot=<0.0443 0.6596 0.7489 0.0462>}
 dyn_hair_beads_l_02 {parent=dyn_hair_beads_l_01, loc=<6.4082 0 0>, rot=<0 0 0 1>}
 dyn_hair_beads_r_bot {parent=head, loc=<5.966 5.9929 6.1465>, rot=<0.1176 0.342 0.931 0.0498>}
 dyn_hair_ribbon_C_01 {parent=head, loc=<5.9771 4.6481 6.3493>, rot=<-0.0022 0.6344 0.7721 0.0378>}
 dyn_hair_ribbon_C_02 {parent=dyn_hair_ribbon_C_01, loc=<6.2029 0 0>, rot=<-0.0582 0.0216 -0.027 -0.9977>}
 dyn_hair_ribbon_C_03 {parent=dyn_hair_ribbon_C_02, loc=<7.0851 0 0>, rot=<0 0 0 1>}
 dyn_hair_beads_r_base {parent=head, loc=<15.7994 5.6477 4.0069>, rot=<0.4431 0.3912 0.776 0.2201>}
 dyn_hair_beads_r_top_01 {parent=dyn_hair_beads_r_base, loc=<1.016 0 0>, rot=<-0.0611 -0.2873 0.0569 -0.9542>}
 dyn_hair_beads_r_top_02 {parent=dyn_hair_beads_r_top_01, loc=<3.544 0 0>, rot=<0.0797 -0.0855 0.0105 -0.9931>}
 dyn_hair_dread_fr_l {parent=head, loc=<5.9198 6.1176 -5.9693>, rot=<-0.0371 -0.9573 -0.2587 -0.1234>}
 dyn_hair_side_l_base {parent=head, loc=<7.6116 -4.4173 -6.5339>, rot=<0.0301 -0.95 0.2992 -0.0844>}
 dyn_hair_side_l_01 {parent=dyn_hair_side_l_base, loc=<0 0 0>, rot=<0.0707 0.0169 -0.0532 -0.9959>}
 dyn_hair_side_l_02 {parent=dyn_hair_side_l_01, loc=<7.4988 -0.0689 -0.1894>, rot=<-0.0093 0.0362 -0.0105 -0.9992>}
 dyn_hair_side_l_03 {parent=dyn_hair_side_l_02, loc=<7.0907 0 0>, rot=<-0.13 0.0864 0.0114 -0.9877>}
 dyn_hair_bck_l_base {parent=head, loc=<8.2112 -6.5462 -3.8012>, rot=<0.0034 -0.8191 0.5736 -0.0024>}
 dyn_hair_bck_l_01 {parent=dyn_hair_bck_l_base, loc=<0 0 0>, rot=<-0.0456 0.0367 -0.0047 -0.9983>}
 dyn_hair_bck_l_02 {parent=dyn_hair_bck_l_01, loc=<7.8818 0.355 -0.0649>, rot=<-0.0114 0.1036 0.0175 -0.9944>}
 dyn_hair_bck_l_03 {parent=dyn_hair_bck_l_02, loc=<7.2411 0.6335 0>, rot=<-0.0062 0.1707 0.0056 -0.9853>}
 dyn_hair_bck_l_04 {parent=dyn_hair_bck_l_03, loc=<6.3048 0 0>, rot=<0.0038 -0.1713 0.0312 -0.9847>}
 dyn_hair_bck_l_05 {parent=dyn_hair_bck_l_04, loc=<7.105 -0.6192 0.0542>, rot=<0 -0.0038 -0.0435 -0.999>}
 dyn_hair_bck_l_06 {parent=dyn_hair_bck_l_05, loc=<4.1505 -0.98 -1.2721>, rot=<0 0 0 1>}
 dyn_hair_bck_braid_base {parent=dyn_hair_bck_l_base, loc=<0.6835 -2.4496 3.2517>, rot=<0 0.1305 0 -0.9914>}
 dyn_hair_bck_braid_01 {parent=dyn_hair_bck_braid_base, loc=<0 0 0>, rot=<-0.065 0.0129 0.0124 -0.9977>}
 dyn_hair_bck_braid_02 {parent=dyn_hair_bck_braid_01, loc=<6.6908 0.3488 -0.3599>, rot=<-0.0159 -0.0098 0.0083 -0.9998>}
 dyn_hair_bck_braid_03 {parent=dyn_hair_bck_braid_02, loc=<6.0183 -0.3304 0.0847>, rot=<0.0087 -0.0002 0 -1>}
 dyn_hair_bck_braid_04 {parent=dyn_hair_bck_braid_03, loc=<5.2308 0.0802 0.3425>, rot=<0 0 0 1>}
 dyn_hair_bck_r_base {parent=head, loc=<8.2169 -7.2331 2.6589>, rot=<0.0025 -0.6088 0.7933 -0.0033>}
 dyn_hair_bck_r_01 {parent=dyn_hair_bck_r_base, loc=<0 0 0>, rot=<0.1257 0.0926 -0.0002 -0.9877>}
 dyn_hair_bck_r_02 {parent=dyn_hair_bck_r_01, loc=<7.6577 -1.0219 -0.3899>, rot=<0.003 0.0115 -0.0123 -0.9999>}
 dyn_hair_bck_r_03 {parent=dyn_hair_bck_r_02, loc=<6.4631 0 0>, rot=<0 0.1224 0 -0.9925>}
 dyn_hair_bck_r_04 {parent=dyn_hair_bck_r_03, loc=<5.7296 0 0>, rot=<0 -0.1068 0 -0.9943>}
 dyn_hair_bck_r_05 {parent=dyn_hair_bck_r_04, loc=<6.8258 0 0>, rot=<0 0 0 1>}
 dyn_hair_bck_ribbon_l_base {parent=dyn_hair_bck_r_01, loc=<2.707 2.1407 1.143>, rot=<-0.148 0.0813 0.0126 -0.9856>}
 dyn_hair_bck_ribbon_l_01 {parent=dyn_hair_bck_ribbon_l_base, loc=<0 0 0>, rot=<0.1105 0.0149 -0.0134 -0.9937>}
 dyn_hair_bck_ribbon_l_02 {parent=dyn_hair_bck_ribbon_l_01, loc=<5.8513 -1.0166 0.1247>, rot=<0.0307 -0.0198 -0.0112 -0.9993>}
 dyn_hair_bck_ribbon_l_03 {parent=dyn_hair_bck_ribbon_l_02, loc=<6.1313 -0.0311 0.5547>, rot=<-0.0021 -0.1637 -0.0201 -0.9863>}
 dyn_hair_bck_ribbon_l_04 {parent=dyn_hair_bck_ribbon_l_03, loc=<5.7105 -0.053 2.0478>, rot=<0 0 0 1>}
 dyn_hair_bck_ribbon_l_05 {parent=dyn_hair_bck_ribbon_l_04, loc=<7.6355 -0.1906 0.6671>, rot=<0 0 0 1>}
 dyn_hair_bck_ribbon_r_base {parent=dyn_hair_bck_r_01, loc=<1.1328 -1.5567 1.3013>, rot=<-0.1146 -0.0766 0.1134 -0.9839>}
 dyn_hair_bck_ribbon_r_01 {parent=dyn_hair_bck_ribbon_r_base, loc=<0 0 0>, rot=<0.0512 0.1279 -0.0201 -0.9903>}
 dyn_hair_bck_ribbon_r_02 {parent=dyn_hair_bck_ribbon_r_01, loc=<6.8056 -0.4685 -0.5657>, rot=<-0.0124 0.0317 -0.0692 -0.997>}
 dyn_hair_bck_ribbon_r_03 {parent=dyn_hair_bck_ribbon_r_02, loc=<6.1912 -1.1391 0.0187>, rot=<-0.0214 -0.0616 -0.015 -0.9978>}
 dyn_hair_bck_ribbon_r_04 {parent=dyn_hair_bck_ribbon_r_03, loc=<6.185 -0.0251 0.8995>, rot=<0 0 0 1>}
 dyn_hair_bck_ribbon_r_05 {parent=dyn_hair_bck_ribbon_r_04, loc=<6.4593 0.2738 -0.2708>, rot=<0 0 0 1>}
 dyn_hair_side_r_base {parent=head, loc=<8.1916 -4.1737 7.2962>, rot=<0.1624 -0.3473 0.9235 -0.0149>}
 dyn_hair_side_r_01 {parent=dyn_hair_side_r_base, loc=<0 0 0>, rot=<0.0089 0.0019 -0.0034 -1>}
 dyn_hair_side_r_02 {parent=dyn_hair_side_r_01, loc=<8.0718 -0.5243 -0.5356>, rot=<0.1009 -0.0393 -0.0704 -0.9916>}
 dyn_hair_side_r_03 {parent=dyn_hair_side_r_02, loc=<7.4467 0 0>, rot=<0.1907 0.3089 -0.247 -0.8984>}
 dyn_hair_side_r_04 {parent=dyn_hair_side_r_03, loc=<5.7678 0.0999 -0.0123>, rot=<0.0916 -0.1587 0.0441 -0.9821>}
 faceAttach {parent=head, loc=<-4.146 -0.8545 0>, rot=<0.0029 -0.7071 -0.0029 -0.7071>}
 C_brow_mid {parent=faceAttach, loc=<0 10.9866 10.1997>, rot=<0 0 0 -1>}
 L_brow_mid {parent=faceAttach, loc=<2.4652 10.3419 10.5343>, rot=<0.0743 -0.0194 0.1982 -0.9771>}
 L_brow_outer {parent=faceAttach, loc=<4.6642 9.3462 10.4432>, rot=<-0.0341 -0.0355 0.3168 -0.9472>}
 R_brow_mid {parent=faceAttach, loc=<-2.4652 10.3419 10.5343>, rot=<0.0743 0.0194 -0.1982 -0.9771>}
 R_brow_outer {parent=faceAttach, loc=<-4.6642 9.3462 10.4432>, rot=<-0.0341 0.0355 -0.3168 -0.9472>}
 clavicle_l {parent=spine_05, loc=<4.3575 -0.1466 -2.3223>, rot=<0.0465 -0.7289 -0.009 -0.6829>}
 upperarm_l {parent=clavicle_l, loc=<15.5111 0 0>, rot=<0.0226 -0.3234 -0.0088 -0.9459>}
 clavicle_r {parent=spine_05, loc=<4.3575 -0.1466 2.3223>, rot=<0.7289 0.0465 -0.6829 0.009>}
 upperarm_r {parent=clavicle_r, loc=<-15.5111 0 0>, rot=<0.0226 -0.3234 -0.0088 -0.9459>}
]
variables [
 DeltaTest {type=double, default=0}
 Head Back Drastic {type=double, default=0}
 "Head Up " {type=double, default=0}
 Pelvis Low {type=double, default=0}
]
