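# Physics asset for the face accessory part of the male SpringBreakTrip cosmetic body, setting up physics for its hair and hair beads.
# Capsules cover the upper spine, neck, head, clavicles and upper arms, joined by constraints with free swing and twist.
# Front, back and side hair chains and left-side beads descend from the head, mostly with limited swing and near-zero twist.
# Extra free-angle constraints link some neighboring strands, including the two back-hair tips in both directions.
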PhysicsAsset /BRCosmetics/Characters/Player/Male/Medium/Bodies/M_MED_SpringBreakTrip/Meshes/Parts/M_MED_SpringBreakTrip_FaceAcc_Physics
bodies {
 clavicle_l [{shapeType=Capsule, length=10, radius=5.5, shapeName=clavicle_l_capsule, center=<9.0763 0 0.4682>, rotation=<rot 90 0 0>}]
 clavicle_r [{shapeType=Capsule, length=10, radius=5.5, shapeName=clavicle_r_capsule, center=<-9.0763 0 -1.8095>, rotation=<rot -90 0 0>}]
 dyn_hair_bck_l_01 [{shapeType=Capsule, length=3, radius=2, shapeName=dyn_hair_bck_l_01_capsule, center=<2.287 0 0>, rotation=<rot 90 -179.9999 179.9999>}]
 dyn_hair_bck_l_02 [{shapeType=Capsule, length=2, radius=4.5, shapeName=dyn_hair_bck_l_02_capsule, center=<2.287 0.6805 3.9748>, rotation=<rot 90 180 179.9999>}]
 dyn_hair_bck_l_03 [{shapeType=Capsule, length=2.5, radius=4, shapeName=dyn_hair_bck_l_03_capsule, center=<2.287 -2.6323 2.6917>, rotation=<rot 90 -180 179.9999>}]
 dyn_hair_bck_l_04 [{shapeType=Capsule, length=4, radius=4.5, shapeName=dyn_hair_bck_l_04_capsule, center=<2.637 -0.2236 2.729>, rotation=<rot 0 0 90>}]
 dyn_hair_bck_l_05 [{shapeType=Capsule, length=4, radius=5, shapeName=dyn_hair_bck_l_05_capsule, center=<-0.4196 -1.1511 -0.2323>, rotation=<rot 0 0 90>}]
 dyn_hair_bck_l_06 [{
   shapeType=Capsule
   length=4.2124
   radius=3.5412
   shapeName=dyn_hair_bck_l_06_capsule
   center=<7.2933 0 -1.1115>
   rotation=<rot 84.9999 0 -0.0001>
  }]
 dyn_hair_bck_r_01 [{shapeType=Capsule, length=0.5, radius=3, shapeName=dyn_hair_bck_r_01_capsule, center=<2.5488 0 0>, rotation=<rot 90 6.7214 6.7214>}]
 dyn_hair_bck_r_02 [{shapeType=Capsule, length=1, radius=4.5, shapeName=dyn_hair_bck_r_02_capsule, center=<2.287 0 1.9728>, rotation=<rot 90 -180 -180>}]
 dyn_hair_bck_r_03 [{shapeType=Capsule, length=1.5, radius=4, shapeName=dyn_hair_bck_r_03_capsule, center=<2.287 0 2.7222>, rotation=<rot 90 180 -180>}]
 dyn_hair_bck_r_04 [{shapeType=Capsule, length=2, radius=4.5, shapeName=dyn_hair_bck_r_04_capsule, center=<2.287 0 2.6207>, rotation=<rot 90 180 -180>}]
 dyn_hair_bck_r_05 [{shapeType=Capsule, length=11, radius=5, shapeName=dyn_hair_bck_r_05_capsule, center=<9.8391 0 -0.1402>, rotation=<rot 84.9999 0 0>}]
 dyn_hair_beads_l_01 [{shapeType=Capsule, length=5, radius=1, shapeName=dyn_hair_beads_l_01_capsule, center=<4 0 0>, rotation=<rot 90 0 0>}]
 dyn_hair_beads_l_02 [{shapeType=Capsule, length=6, radius=2.5, shapeName=dyn_hair_beads_l_02_capsule, center=<6 0 2>, rotation=<rot 0 90 95>}]
 dyn_hair_fr_l_01 [{shapeType=Capsule, length=0, radius=1, shapeName=dyn_hair_fr_l_01_capsule}]
 dyn_hair_fr_l_02 [{shapeType=Capsule, length=3, radius=2.5, shapeName=dyn_hair_fr_l_02_capsule, center=<2.287 0 0>, rotation=<rot 90 180 179.9999>}]
 dyn_hair_fr_l_03 [{shapeType=Capsule, length=3.5, radius=2, shapeName=dyn_hair_fr_l_03_capsule, center=<2.287 0 0.7746>, rotation=<rot 90 -178.3124 -178.3124>}]
 dyn_hair_fr_l_04 [{shapeType=Capsule, length=3, radius=2.5, shapeName=dyn_hair_fr_l_04_capsule, center=<2.287 0 0.7302>, rotation=<rot 90 179.9605 179.9604>}]
 dyn_hair_fr_l_05 [{shapeType=Capsule, length=5, radius=3.5, shapeName=dyn_hair_fr_l_05_capsule, center=<4.607 0 0.425>, rotation=<rot 90 179.9605 179.9604>}]
 dyn_hair_fr_l_06 [{shapeType=Capsule, length=3, radius=2.5, shapeName=dyn_hair_fr_l_06_capsule, center=<2.287 -0.6084 0>, rotation=<rot 90 -179.789 -179.7894>}]
 dyn_hair_fr_l_07 [{
   shapeType=Capsule
   length=2.5
   radius=2.5
   shapeName=dyn_hair_fr_l_07_capsule
   center=<1.1067 -0.3224 0.2528>
   rotation=<rot 90 180 179.9996>
  }]
 dyn_hair_fr_r_01 [{shapeType=Capsule, length=4.5, radius=2, shapeName=dyn_hair_fr_r_01_capsule}]
 dyn_hair_fr_r_02 [{shapeType=Capsule, length=2.5, radius=3, shapeName=dyn_hair_fr_r_02_capsule, center=<2.287 0 0.634>, rotation=<rot 90 179.9208 179.9209>}]
 dyn_hair_fr_r_03 [{shapeType=Capsule, length=15, radius=3, shapeName=dyn_hair_fr_r_03_capsule, center=<2.287 0 0>, rotation=<rot 90 180 -180>}]
 dyn_hair_fr_r_04 [{shapeType=Capsule, length=3, radius=2.5, shapeName=dyn_hair_fr_r_04_capsule, center=<2.287 0 0>, rotation=<rot 90 0 0>}]
 dyn_hair_fr_r_05 [{shapeType=Capsule, length=4, radius=4.5, shapeName=dyn_hair_fr_r_05_capsule, center=<3.7043 0 -0.4424>, rotation=<rot 90 180 -180>}]
 dyn_hair_fr_r_06 [{shapeType=Capsule, length=3, radius=3.5, shapeName=dyn_hair_fr_r_06_capsule, center=<2.287 0 0>, rotation=<rot 90 -179.9998 -179.9995>}]
 dyn_hair_side_l_01 [{shapeType=Capsule, length=0, radius=1, shapeName=dyn_hair_side_l_01_capsule}]
 dyn_hair_side_l_02 [{
   shapeType=Capsule
   length=2.5
   radius=3
   shapeName=dyn_hair_side_l_02_capsule
   center=<2.5059 0 -1.1706>
   rotation=<rot 90 -179.9999 179.9997>
  }]
 dyn_hair_side_l_03 [{
   shapeType=Capsule
   length=2.5
   radius=3
   shapeName=dyn_hair_side_l_03_capsule
   center=<4.9182 0.1507 0.2568>
   rotation=<rot 80 -169.9995 -179.9999>
  }]
 dyn_hair_side_l_base [{shapeType=Capsule, length=0, radius=1, shapeName=dyn_hair_side_l_base_capsule, rotation=<rot -90 0 0>}]
 dyn_hair_side_r_01 [{shapeType=Capsule, length=0, radius=1, shapeName=dyn_hair_side_r_01_capsule}]
 dyn_hair_side_r_02 [{
   shapeType=Capsule
   length=4.5
   radius=3
   shapeName=dyn_hair_side_r_02_capsule
   center=<-0.2552 -1.7334 2.5846>
   rotation=<rot 90 180 -167.7843>
  }]
 dyn_hair_side_r_03 [{shapeType=Capsule, length=3, radius=3.5, shapeName=dyn_hair_side_r_03_capsule, center=<1.7 -1.0171 0.2301>, rotation=<rot 90 0 0>}]
 dyn_hair_side_r_04 [{
   shapeType=Capsule
   length=2.5
   radius=3.5
   shapeName=dyn_hair_side_r_04_capsule
   center=<2.287 -1.0824 -0.5115>
   rotation=<rot 90 -5.0512 -5.0512>
  }]
 head [{shapeType=Capsule, length=7, radius=8, shapeName=head_capsule, center=<5.3411 2.8582 0>, rotation=<rot 0 -90.0003 -90>}]
 neck_01 [{shapeType=Capsule, length=10.5, radius=5.5, shapeName=neck_01_capsule, rotation=<rot 0 -0.0003 0>}]
 neck_02 [{shapeType=Capsule, length=2.5, radius=5.5, shapeName=neck_02_capsule, rotation=<rot 90 164.8051 164.8053>}]
 spine_04 [{shapeType=Capsule, length=13.005, radius=11, shapeName=spine_04_capsule, center=<0 -0.5392 0>, rotation=<rot 0 -90 0>}]
 spine_05 [{shapeType=Capsule, length=15, radius=8, shapeName=spine_05_capsule, center=<2.5294 0.4739 0>}]
 upperarm_l [{shapeType=Capsule, length=23.5, radius=5.5, shapeName=upperarm_l_capsule, center=<8.3369 0 0>, rotation=<rot 90 0 0>}]
 upperarm_r [{shapeType=Capsule, length=25, radius=7, shapeName=upperarm_r_capsule, center=<-13 0 0>, rotation=<rot -90 0 0>}]
}
constraints [
 {bone1Name=dyn_hair_side_r_04, bone2Name=dyn_hair_side_r_03, swing1LimitDegrees=20, swing2LimitDegrees=20, twistLimitDegrees=0}
 {
  bone1Name=spine_05
  bone2Name=spine_04
  swing1LimitDegrees=45
  swing1Motion=Free
  swing2LimitDegrees=45
  swing2Motion=Free
  twistLimitDegrees=45
  twistMotion=Free
 }
 {
  bone1Name=clavicle_l
  bone2Name=spine_05
  swing1LimitDegrees=45
  swing1Motion=Free
  swing2LimitDegrees=45
  swing2Motion=Free
  twistLimitDegrees=45
  twistMotion=Free
 }
 {
  bone1Name=upperarm_l
  bone2Name=clavicle_l
  swing1LimitDegrees=45
  swing1Motion=Free
  swing2LimitDegrees=45
  swing2Motion=Free
  twistLimitDegrees=45
  twistMotion=Free
 }
 {
  bone1Name=clavicle_r
  bone2Name=spine_05
  swing1LimitDegrees=45
  swing1Motion=Free
  swing2LimitDegrees=45
  swing2Motion=Free
  twistLimitDegrees=45
  twistMotion=Free
 }
 {
  bone1Name=upperarm_r
  bone2Name=clavicle_r
  swing1LimitDegrees=45
  swing1Motion=Free
  swing2LimitDegrees=45
  swing2Motion=Free
  twistLimitDegrees=45
  twistMotion=Free
 }
 {
  bone1Name=neck_01
  bone2Name=spine_05
  swing1LimitDegrees=45
  swing1Motion=Free
  swing2LimitDegrees=45
  swing2Motion=Free
  twistLimitDegrees=45
  twistMotion=Free
 }
 {
  bone1Name=neck_02
  bone2Name=neck_01
  swing1LimitDegrees=45
  swing1Motion=Free
  swing2LimitDegrees=45
  swing2Motion=Free
  twistLimitDegrees=45
  twistMotion=Free
 }
 {
  bone1Name=head
  bone2Name=neck_02
  swing1LimitDegrees=45
  swing1Motion=Free
  swing2LimitDegrees=45
  swing2Motion=Free
  twistLimitDegrees=45
  twistMotion=Free
 }
 {bone1Name=dyn_hair_bck_l_01, bone2Name=head, swing1LimitDegrees=45, swing2LimitDegrees=45, twistLimitDegrees=0}
 {bone1Name=dyn_hair_bck_l_02, bone2Name=dyn_hair_bck_l_01, swing1LimitDegrees=20, swing2LimitDegrees=15, twistLimitDegrees=0}
 {bone1Name=dyn_hair_bck_l_03, bone2Name=dyn_hair_bck_l_02, swing1LimitDegrees=20, swing2LimitDegrees=15, twistLimitDegrees=0}
 {bone1Name=dyn_hair_bck_l_04, bone2Name=dyn_hair_bck_l_03, swing1LimitDegrees=20, swing2LimitDegrees=15, twistLimitDegrees=0}
 {bone1Name=dyn_hair_bck_l_05, bone2Name=dyn_hair_bck_l_04, swing1LimitDegrees=20, swing2LimitDegrees=15, twistLimitDegrees=0}
 {bone1Name=dyn_hair_bck_l_06, bone2Name=dyn_hair_bck_l_05, swing1LimitDegrees=25, swing2LimitDegrees=25, twistLimitDegrees=0}
 {
  bone1Name=dyn_hair_bck_l_06
  bone2Name=dyn_hair_bck_r_05
  swing1LimitDegrees=45
  swing1Motion=Free
  swing2LimitDegrees=45
  swing2Motion=Free
  twistLimitDegrees=45
  twistMotion=Free
 }
 {
  bone1Name=dyn_hair_bck_r_05
  bone2Name=dyn_hair_bck_l_06
  swing1LimitDegrees=45
  swing1Motion=Free
  swing2LimitDegrees=45
  swing2Motion=Free
  twistLimitDegrees=45
  twistMotion=Free
 }
 {bone1Name=dyn_hair_bck_r_01, bone2Name=head, swing1LimitDegrees=45, swing2LimitDegrees=45, twistLimitDegrees=0}
 {bone1Name=dyn_hair_bck_r_02, bone2Name=dyn_hair_bck_r_01, swing1LimitDegrees=20, swing2LimitDegrees=20, twistLimitDegrees=0}
 {bone1Name=dyn_hair_bck_r_03, bone2Name=dyn_hair_bck_r_02, swing1LimitDegrees=20, swing2LimitDegrees=20, twistLimitDegrees=0}
 {
  bone1Name=dyn_hair_side_r_04
  bone2Name=dyn_hair_bck_r_04
  swing1LimitDegrees=45
  swing1Motion=Free
  swing2LimitDegrees=45
  swing2Motion=Free
  twistLimitDegrees=45
  twistMotion=Free
 }
 {bone1Name=dyn_hair_bck_r_04, bone2Name=dyn_hair_bck_r_03, swing1LimitDegrees=20, swing2LimitDegrees=20, twistLimitDegrees=0}
 {bone1Name=dyn_hair_bck_r_05, bone2Name=dyn_hair_bck_r_04, swing1LimitDegrees=20, swing2LimitDegrees=20, twistLimitDegrees=0}
 {bone1Name=dyn_hair_beads_l_01, bone2Name=head, swing1LimitDegrees=20, swing2LimitDegrees=20, twistLimitDegrees=0}
 {bone1Name=dyn_hair_beads_l_02, bone2Name=dyn_hair_beads_l_01, swing1LimitDegrees=20, swing2LimitDegrees=20, twistLimitDegrees=0}
 {bone1Name=dyn_hair_fr_l_01, bone2Name=head, swing1LimitDegrees=10, swing2LimitDegrees=10, twistLimitDegrees=0}
 {bone1Name=dyn_hair_fr_l_02, bone2Name=dyn_hair_fr_l_01, swing1LimitDegrees=15, swing2LimitDegrees=25, twistLimitDegrees=0}
 {bone1Name=dyn_hair_fr_l_03, bone2Name=dyn_hair_fr_l_02, swing1LimitDegrees=20, swing2LimitDegrees=40, twistLimitDegrees=0}
 {
  bone1Name=dyn_hair_fr_l_03
  bone2Name=dyn_hair_fr_l_07
  swing1LimitDegrees=45
  swing1Motion=Free
  swing2LimitDegrees=45
  swing2Motion=Free
  twistLimitDegrees=45
  twistMotion=Free
 }
 {bone1Name=dyn_hair_fr_l_04, bone2Name=dyn_hair_fr_l_03, swing1LimitDegrees=20, swing2LimitDegrees=40, twistLimitDegrees=0}
 {bone1Name=dyn_hair_fr_l_05, bone2Name=dyn_hair_fr_l_04, swing1LimitDegrees=20, swing2LimitDegrees=40, twistLimitDegrees=0}
 {bone1Name=dyn_hair_fr_l_06, bone2Name=dyn_hair_fr_l_01, swing1LimitDegrees=20, swing2LimitDegrees=40, twistLimitDegrees=0}
 {bone1Name=dyn_hair_fr_l_07, bone2Name=dyn_hair_fr_l_06, swing1LimitDegrees=15, swing2LimitDegrees=15, twistLimitDegrees=0}
 {bone1Name=dyn_hair_fr_r_01, bone2Name=head, swing1LimitDegrees=45, swing2LimitDegrees=45, twistLimitDegrees=0}
 {bone1Name=dyn_hair_fr_r_02, bone2Name=dyn_hair_fr_r_01, swing1LimitDegrees=10, swing2LimitDegrees=10, twistLimitDegrees=0}
 {bone1Name=dyn_hair_fr_r_03, bone2Name=dyn_hair_fr_r_02, swing1LimitDegrees=20, swing2LimitDegrees=20, twistLimitDegrees=0}
 {
  bone1Name=dyn_hair_fr_r_06
  bone2Name=dyn_hair_fr_r_03
  swing1LimitDegrees=45
  swing1Motion=Free
  swing2LimitDegrees=45
  swing2Motion=Free
  twistLimitDegrees=45
  twistMotion=Free
 }
 {bone1Name=dyn_hair_fr_r_04, bone2Name=dyn_hair_fr_r_03, swing1LimitDegrees=20, swing2LimitDegrees=40, twistLimitDegrees=0}
 {bone1Name=dyn_hair_fr_r_05, bone2Name=dyn_hair_fr_r_04, swing1LimitDegrees=20, swing2LimitDegrees=40, twistLimitDegrees=0}
 {bone1Name=dyn_hair_fr_r_06, bone2Name=dyn_hair_fr_r_02, swing1LimitDegrees=20, swing2LimitDegrees=40, twistLimitDegrees=0}
 {
  bone1Name=dyn_hair_side_l_base
  bone2Name=head
  swing1LimitDegrees=45
  swing1Motion=Free
  swing2LimitDegrees=45
  swing2Motion=Free
  twistLimitDegrees=45
  twistMotion=Free
 }
 {bone1Name=dyn_hair_side_l_01, bone2Name=dyn_hair_side_l_base, swing1LimitDegrees=45, swing2LimitDegrees=45, twistLimitDegrees=0}
 {bone1Name=dyn_hair_side_l_02, bone2Name=dyn_hair_side_l_01, swing1LimitDegrees=20, swing2LimitDegrees=20, twistLimitDegrees=0}
 {bone1Name=dyn_hair_side_l_03, bone2Name=dyn_hair_side_l_02, swing1LimitDegrees=10, swing2LimitDegrees=40, twistLimitDegrees=0}
 {bone1Name=dyn_hair_side_r_01, bone2Name=head, swing1LimitDegrees=20, swing2LimitDegrees=20, twistLimitDegrees=0}
 {bone1Name=dyn_hair_side_r_02, bone2Name=dyn_hair_side_r_01, swing1LimitDegrees=20, swing2LimitDegrees=20, twistLimitDegrees=1}
 {bone1Name=dyn_hair_side_r_03, bone2Name=dyn_hair_side_r_02, swing1LimitDegrees=30, swing2LimitDegrees=30, twistLimitDegrees=0}
]
