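# AnimBlueprint for the SpringBreakTrip FaceAcc mesh, adding secondary motion to hair, braids, ribbons and beads.
# Poses copied from parent and body meshes feed a Control Rig.
# The full path runs constraints and RigidBody, swapping physics assets when skydiving, then ribbon constraints, AnimDynamics, head-up offsets and wave rotations.
# Wearing a back cape instead outputs a RigidBody pose with hair constrained toward spine, head and clavicles.
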AnimBlueprint /BRCosmetics/Characters/Player/Male/Medium/Bodies/M_MED_SpringBreakTrip/Meshes/Parts/M_MED_SpringBreakTrip_FaceAcc_AnimBP
target_skeleton=&/M_MED_SpringBreakTrip_FaceAcc_Skeleton
graphs [
 AnimGraph {nodes [
   "Constraint - dyn_hair_fr_l_02 (PP)"!(ComponentPose, Alpha=0.5, ConstraintWeights_0=0.2, ConstraintWeights_1:float=0.8) -> (Pose) {
    alphaBoolBlend {blendOption=Linear}
    boneToModify {boneName=dyn_hair_fr_l_02}
    constraintSetup [{targetBone {boneName=spine_05}}, {targetBone {boneName=clavicle_l}}]
    constraintWeights [1,1]
   }
   n1 = "Constraint - dyn_hair_bck_l_03 (PP)"!(ComponentPose, Alpha=0.5, ConstraintWeights_0=0.2, ConstraintWeights_1:float=0.8) -> (Pose) {
    alphaBoolBlend {blendOption=Linear}
    boneToModify {boneName=dyn_hair_bck_l_03}
    constraintSetup [{targetBone {boneName=spine_05}}, {targetBone {boneName=clavicle_l}}]
    constraintWeights [1,1]
   }
   n2 = "Constraint - dyn_hair_bck_l_04 (PP)"!(ComponentPose=@n1, Alpha=0.5, ConstraintWeights_0=0.2, ConstraintWeights_1:float=0.8) -> (Pose) {
    alphaBoolBlend {blendOption=Linear}
    boneToModify {boneName=dyn_hair_bck_l_04}
    constraintSetup [{targetBone {boneName=spine_05}}, {targetBone {boneName=clavicle_l}}]
    constraintWeights [1,1]
   }
   n3 = "Constraint - dyn_hair_bck_l_05 (PP)"!(ComponentPose=@n2, Alpha=0.5, ConstraintWeights_0=0.2, ConstraintWeights_1:float=0.8) -> (Pose) {
    alphaBoolBlend {blendOption=Linear}
    boneToModify {boneName=dyn_hair_bck_l_05}
    constraintSetup [{targetBone {boneName=spine_05}}, {targetBone {boneName=clavicle_l}}]
    constraintWeights [1,1]
   }
   "Constraint - dyn_hair_bck_l_06 (PP)"!(ComponentPose=@n3, Alpha=1, ConstraintWeights_0=0.2, ConstraintWeights_1:float=0.8) -> (Pose) {
    alphaBoolBlend {blendOption=Linear}
    boneToModify {boneName=dyn_hair_bck_l_06}
    constraintSetup [{targetBone {boneName=spine_05}}, {targetBone {boneName=clavicle_l}}]
    constraintWeights [1,1]
   }
   Output Pose(Result=Blend Poses by bool(
     BlendPose_0=Use cached pose 'constraints_Cache'()
     BlendPose_1=Use cached pose 'All_sim_Cache'()
     bActiveValue=Get isWearingBackCape() -> (isWearingBackCape:bool)
    ) {blendType=Linear}) {layerGroup=None}
   n4 = Get Body_Mesh() -> (Body_Mesh:Skeletal Mesh Component Budgeted Object Reference)
   n5 = Control Rig(
    Source=Layered blend per bone(
     BasePose=Copy Pose From Mesh() {bCopyCurves=true, bUseAttachedParent=true}
     BlendPoses_0=Copy Pose From Mesh(SourceMeshComponent=@n4) {bUseAttachedParent=true}
    ) {layerSetup [{branchFilters [{boneName=clavicle_l}, {boneName=clavicle_r}]}]}
    Alpha=0.75
   ) {
    alphaBoolBlend {blendOption=Linear}
    controlRigAssetReference {blueprintRigClass=&/M_MED_SpringBreakTrip_FaceAcc_CtrlRig.M_MED_SpringBreakTrip_FaceAcc_CtrlRig_C}
   }
   Save cached pose 'CtrlRigCache'(Pose=@n5) {cacheName=CtrlRigCache}
   n6 = RigidBody!(
    ComponentPose=Local To Component(LocalPose=Use cached pose 'CtrlRigCache'())
    Alpha
    OverrideWorldGravity:Vector=Get GravityOverride()
    ExternalForce
   ) -> (Pose) {
    bClampLinearTranslationLimitToRefPose=true
    bOverrideWorldGravity=true
    overridePhysicsAsset=&/M_MED_SpringBreakTrip_FaceAcc_Physics
    simulationSpace=BaseBoneSpace
    alphaBoolBlend {blendOption=Linear}
    baseBoneRef {boneName=neck_02}
    componentAppliedLinearAccClamp=<3000 3000 3000>
    componentLinearAccScale=<0.3 0.3 0.3>
    componentLinearVelScale=<0.3 0.3 0.3>
    overrideWorldGravity {z=-980}
    simSpaceSettings {maxLinearVelocity=0.3, worldAlpha=0.35}
   }
   n7 = "Constraint - dyn_hair_bck_braid_01 (PP)"!(ComponentPose=@n6, Alpha=0.5, ConstraintWeights_0=0.2, ConstraintWeights_1:float=0.8) -> (Pose) {
    alphaBoolBlend {blendOption=Linear}
    boneToModify {boneName=dyn_hair_bck_braid_01}
    constraintSetup [{targetBone {boneName=spine_05}}, {targetBone {boneName=head}}]
    constraintWeights [1,1]
   }
   n8 = "Constraint - dyn_hair_bck_braid_02 (PP)"!(ComponentPose=@n7, Alpha=0.5, ConstraintWeights_0=0.2, ConstraintWeights_1:float=0.8) -> (Pose) {
    alphaBoolBlend {blendOption=Linear}
    boneToModify {boneName=dyn_hair_bck_braid_02}
    constraintSetup [{targetBone {boneName=spine_05}}, {targetBone {boneName=clavicle_l}}]
    constraintWeights [1,1]
   }
   n9 = "Constraint - dyn_hair_bck_braid_03 (PP)"!(ComponentPose=@n8, Alpha=0.5, ConstraintWeights_0=0.2, ConstraintWeights_1:float=0.8) -> (Pose) {
    alphaBoolBlend {blendOption=Linear}
    boneToModify {boneName=dyn_hair_bck_braid_03}
    constraintSetup [{targetBone {boneName=spine_05}}, {targetBone {boneName=clavicle_l}}]
    constraintWeights [1,1]
   }
   n10 = "Constraint - dyn_hair_bck_braid_04 (PP)"!(ComponentPose=@n9, Alpha=0.5, ConstraintWeights_0=0.2, ConstraintWeights_1:float=0.8) -> (Pose) {
    alphaBoolBlend {blendOption=Linear}
    boneToModify {boneName=dyn_hair_bck_braid_04}
    constraintSetup [{targetBone {boneName=spine_05}}, {targetBone {boneName=clavicle_l}}]
    constraintWeights [1,1]
   }
   n11 = "Constraint - dyn_hair_bck_l_01 (PP)"!(ComponentPose=@n10, Alpha=0.4, ConstraintWeights_0=0.5, ConstraintWeights_1:float=0.5) -> (Pose) {
    alphaBoolBlend {blendOption=Linear}
    boneToModify {boneName=dyn_hair_bck_l_01}
    constraintSetup [{targetBone {boneName=spine_05}}, {targetBone {boneName=clavicle_l}}]
    constraintWeights [1,1]
   }
   n12 = "Constraint - dyn_hair_bck_l_02 (PP)"!(ComponentPose=@n11, Alpha=0.8, ConstraintWeights_0=0.8, ConstraintWeights_1:float=0.2) -> (Pose) {
    alphaBoolBlend {blendOption=Linear}
    boneToModify {boneName=dyn_hair_bck_l_02}
    constraintSetup [{targetBone {boneName=spine_05}}, {targetBone {boneName=clavicle_l}}]
    constraintWeights [1,1]
   }
   n13 = "Constraint - dyn_hair_side_l_01 (PP)"!(ComponentPose=@n12, Alpha=0.5, ConstraintWeights_0=0.8, ConstraintWeights_1:float=0.2) -> (Pose) {
    alphaBoolBlend {blendOption=Linear}
    boneToModify {boneName=dyn_hair_side_l_01}
    constraintSetup [{targetBone {boneName=spine_05}}, {targetBone {boneName=clavicle_l}}]
    constraintWeights [1,1]
   }
   n14 = "Constraint - dyn_hair_side_l_02 (PP)"!(ComponentPose=@n13, Alpha=0.5, ConstraintWeights_0=0.2, ConstraintWeights_1:float=0.8) -> (Pose) {
    alphaBoolBlend {blendOption=Linear}
    boneToModify {boneName=dyn_hair_side_l_02}
    constraintSetup [{targetBone {boneName=spine_05}}, {targetBone {boneName=clavicle_l}}]
    constraintWeights [1,1]
   }
   n15 = "Constraint - dyn_hair_side_l_03 (PP)"!(ComponentPose=@n14, Alpha=1, ConstraintWeights_0=0.2, ConstraintWeights_1:float=0.8) -> (Pose) {
    alphaBoolBlend {blendOption=Linear}
    boneToModify {boneName=dyn_hair_side_l_03}
    constraintSetup [{targetBone {boneName=spine_05}}, {targetBone {boneName=clavicle_l}}]
    constraintWeights [1,1]
   }
   n16 = "Constraint - dyn_hair_bck_r_03 (PP)"!(ComponentPose=@n15, Alpha=0.5, ConstraintWeights_0=0.8, ConstraintWeights_1:float=0.2) -> (Pose) {
    alphaBoolBlend {blendOption=Linear}
    boneToModify {boneName=dyn_hair_bck_r_03}
    constraintSetup [{targetBone {boneName=clavicle_r}}, {targetBone {boneName=spine_05}}]
    constraintWeights [1,1]
   }
   n17 = "Constraint - dyn_hair_bck_r_04 (PP)"!(ComponentPose=@n16, Alpha=0.5, ConstraintWeights_0=0.8, ConstraintWeights_1:float=0.2) -> (Pose) {
    alphaBoolBlend {blendOption=Linear}
    boneToModify {boneName=dyn_hair_bck_r_04}
    constraintSetup [{targetBone {boneName=clavicle_r}}, {targetBone {boneName=spine_05}}]
    constraintWeights [1,1]
   }
   n18 = "Constraint - dyn_hair_bck_r_05 (PP)"!(ComponentPose=@n17, Alpha=0.5, ConstraintWeights_0=0.8, ConstraintWeights_1:float=0.2) -> (Pose) {
    alphaBoolBlend {blendOption=Linear}
    boneToModify {boneName=dyn_hair_bck_r_05}
    constraintSetup [{targetBone {boneName=clavicle_r}}, {targetBone {boneName=spine_05}}]
    constraintWeights [1,1]
   }
   n19 = "Constraint - dyn_hair_side_r_01 (PP)"!(ComponentPose=@n18, Alpha=0.25, ConstraintWeights_0=0.8, ConstraintWeights_1:float=0.2) -> (Pose) {
    alphaBoolBlend {blendOption=Linear}
    boneToModify {boneName=dyn_hair_side_r_01}
    constraintSetup [{targetBone {boneName=spine_05}}, {targetBone {boneName=clavicle_r}}]
    constraintWeights [1,1]
   }
   n20 = "Constraint - dyn_hair_side_r_02 (PP)"!(ComponentPose=@n19, Alpha=0.5, ConstraintWeights_0=0.2, ConstraintWeights_1:float=0.8) -> (Pose) {
    alphaBoolBlend {blendOption=Linear}
    boneToModify {boneName=dyn_hair_side_r_02}
    constraintSetup [{targetBone {boneName=spine_05}}, {targetBone {boneName=clavicle_r}}]
    constraintWeights [1,1]
   }
   n21 = "Constraint - dyn_hair_side_r_03 (PP)"!(ComponentPose=@n20, Alpha=0.75, ConstraintWeights_0=0.8, ConstraintWeights_1:float=0.2) -> (Pose) {
    alphaBoolBlend {blendOption=Linear}
    boneToModify {boneName=dyn_hair_side_r_03}
    constraintSetup [{targetBone {boneName=spine_05}}, {targetBone {boneName=clavicle_r}}]
    constraintWeights [1,1]
   }
   n22 = "Constraint - dyn_hair_side_r_04 (PP)"!(ComponentPose=@n21, Alpha=1, ConstraintWeights_0=0.8, ConstraintWeights_1:float=0.2) -> (Pose) {
    alphaBoolBlend {blendOption=Linear}
    boneToModify {boneName=dyn_hair_side_r_04}
    constraintSetup [{targetBone {boneName=spine_05}}, {targetBone {boneName=clavicle_r}}]
    constraintWeights [1,1]
   }
   n23 = "Constraint - dyn_hair_bck_ribbon_l_01 (PP)"!(ComponentPose=@n22, Alpha=0.3, ConstraintWeights_0=0.8, ConstraintWeights_1:float=0.2) -> (Pose) {
    alphaBoolBlend {blendOption=Linear}
    boneToModify {boneName=dyn_hair_bck_ribbon_l_01}
    constraintSetup [{targetBone {boneName=spine_05}}, {targetBone {boneName=clavicle_r}}]
    constraintWeights [1,1]
   }
   n24 = "Constraint - dyn_hair_bck_ribbon_l_02 (PP)"!(ComponentPose=@n23, Alpha=0.7, ConstraintWeights_0=0.8, ConstraintWeights_1:float=0.2) -> (Pose) {
    alphaBoolBlend {blendOption=Linear}
    boneToModify {boneName=dyn_hair_bck_ribbon_l_02}
    constraintSetup [{targetBone {boneName=spine_05}}, {targetBone {boneName=clavicle_r}}]
    constraintWeights [1,1]
   }
   n25 = "Constraint - dyn_hair_bck_ribbon_l_03 (PP)"!(ComponentPose=@n24, Alpha=0.5, ConstraintWeights_0=0.8, ConstraintWeights_1:float=0.2) -> (Pose) {
    alphaBoolBlend {blendOption=Linear}
    boneToModify {boneName=dyn_hair_bck_ribbon_l_03}
    constraintSetup [{targetBone {boneName=spine_05}}, {targetBone {boneName=clavicle_r}}]
    constraintWeights [1,1]
   }
   n26 = "Constraint - dyn_hair_bck_ribbon_l_04 (PP)"!(ComponentPose=@n25, Alpha=0.5, ConstraintWeights_0=0.8, ConstraintWeights_1:float=0.2) -> (Pose) {
    alphaBoolBlend {blendOption=Linear}
    boneToModify {boneName=dyn_hair_bck_ribbon_l_04}
    constraintSetup [{targetBone {boneName=spine_05}}, {targetBone {boneName=clavicle_r}}]
    constraintWeights [1,1]
   }
   n27 = "Constraint - dyn_hair_bck_ribbon_l_05 (PP)"!(ComponentPose=@n26, Alpha=0.5, ConstraintWeights_0=0.8, ConstraintWeights_1:float=0.2) -> (Pose) {
    alphaBoolBlend {blendOption=Linear}
    boneToModify {boneName=dyn_hair_bck_ribbon_l_05}
    constraintSetup [{targetBone {boneName=spine_05}}, {targetBone {boneName=clavicle_r}}]
    constraintWeights [1,1]
   }
   n28 = "Constraint - dyn_hair_bck_ribbon_r_01 (PP)"!(ComponentPose=@n27, Alpha=0.3, ConstraintWeights_0=0.8, ConstraintWeights_1:float=0.2) -> (Pose) {
    alphaBoolBlend {blendOption=Linear}
    boneToModify {boneName=dyn_hair_bck_ribbon_r_01}
    constraintSetup [{targetBone {boneName=spine_05}}, {targetBone {boneName=clavicle_r}}]
    constraintWeights [1,1]
   }
   n29 = "Constraint - dyn_hair_bck_ribbon_r_02 (PP)"!(ComponentPose=@n28, Alpha=0.7, ConstraintWeights_0=0.8, ConstraintWeights_1:float=0.2) -> (Pose) {
    alphaBoolBlend {blendOption=Linear}
    boneToModify {boneName=dyn_hair_bck_ribbon_r_02}
    constraintSetup [{targetBone {boneName=spine_05}}, {targetBone {boneName=clavicle_r}}]
    constraintWeights [1,1]
   }
   n30 = "Constraint - dyn_hair_bck_ribbon_r_03 (PP)"!(ComponentPose=@n29, Alpha=0.5, ConstraintWeights_0=0.8, ConstraintWeights_1:float=0.2) -> (Pose) {
    alphaBoolBlend {blendOption=Linear}
    boneToModify {boneName=dyn_hair_bck_ribbon_r_03}
    constraintSetup [{targetBone {boneName=spine_05}}, {targetBone {boneName=clavicle_r}}]
    constraintWeights [1,1]
   }
   n31 = "Constraint - dyn_hair_bck_ribbon_r_04 (PP)"!(ComponentPose=@n30, Alpha=0.5, ConstraintWeights_0=0.8, ConstraintWeights_1:float=0.2) -> (Pose) {
    alphaBoolBlend {blendOption=Linear}
    boneToModify {boneName=dyn_hair_bck_ribbon_r_04}
    constraintSetup [{targetBone {boneName=spine_05}}, {targetBone {boneName=clavicle_r}}]
    constraintWeights [1,1]
   }
   n32 = "Constraint - dyn_hair_bck_ribbon_r_05 (PP)"!(ComponentPose=@n31, Alpha=0.5, ConstraintWeights_0=0.8, ConstraintWeights_1:float=0.2) -> (Pose) {
    alphaBoolBlend {blendOption=Linear}
    boneToModify {boneName=dyn_hair_bck_ribbon_r_05}
    constraintSetup [{targetBone {boneName=spine_05}}, {targetBone {boneName=clavicle_r}}]
    constraintWeights [1,1]
   }
   n33 = "Transform (Modify) Bone - Bone: dyn_hair_side_r_02"(ComponentPose=@n32, Alpha=1, Translation="-1,0,0", Rotation="0,0,0", Scale="1,1,1") {
    translationMode=BMM_Additive
    translationSpace=BCS_BoneSpace
    alphaBoolBlend {blendOption=Linear}
    boneToModify {boneName=dyn_hair_side_r_02}
    translation {x=-1}
   }
   n34 = Get headUp() -> (headUp:double)
   n35 = "Transform (Modify) Bone - Bone: dyn_hair_bck_ribbon_l_base"(ComponentPose=@n33, Alpha=@n34, Translation="-3,0,0", Rotation="0,0,0", Scale="1,1,1") {
    rotationMode=BMM_Additive
    rotationSpace=BCS_BoneSpace
    scaleMode=BMM_Additive
    scaleSpace=BCS_BoneSpace
    translationMode=BMM_Additive
    translationSpace=BCS_BoneSpace
    alphaBoolBlend {blendOption=Linear}
    alphaScaleBiasClamp {interpSpeedDecreasing=5}
    boneToModify {boneName=dyn_hair_bck_ribbon_l_base}
    translation {x=-3}
   }
   n36 = "Transform (Modify) Bone - Bone: dyn_hair_bck_ribbon_r_base"(ComponentPose=@n35, Alpha=@n34, Translation="-1,0,4", Rotation="0,0,0", Scale="1,1,1") {
    rotationMode=BMM_Additive
    rotationSpace=BCS_BoneSpace
    scaleMode=BMM_Additive
    scaleSpace=BCS_BoneSpace
    translationMode=BMM_Additive
    translationSpace=BCS_BoneSpace
    alphaBoolBlend {blendOption=Linear}
    alphaScaleBiasClamp {bInterpResult=true, interpSpeedDecreasing=5}
    boneToModify {boneName=dyn_hair_bck_ribbon_r_base}
    translation {x=-1, z=4}
   }
   Save cached pose 'constraints_Cache'(Pose=Component To Local(ComponentPose=@n36)) {cacheName=constraints_Cache}
   n37 = Get headUp()
   n38 = "Transform (Modify) Bone - Bone: dyn_hair_bck_ribbon_l_05"(
    ComponentPose=Local To Component(LocalPose=Use cached pose 'Constrain/Dyncamics Cache'())
    Alpha=@n37
    Translation="0,0,0"
    Rotation="-5,0,0"
    Scale="1,1,1"
   ) {
    rotationMode=BMM_Additive
    rotationSpace=BCS_BoneSpace
    scaleMode=BMM_Additive
    scaleSpace=BCS_BoneSpace
    translationMode=BMM_Additive
    translationSpace=BCS_BoneSpace
    alphaBoolBlend {blendOption=Linear}
    alphaScaleBiasClamp {interpSpeedDecreasing=5}
    boneToModify {boneName=dyn_hair_bck_ribbon_l_05}
    rotation {pitch=-5}
   }
   n39 = "Transform (Modify) Bone - Bone: dyn_hair_bck_ribbon_r_05"(ComponentPose=@n38, Alpha=@n37, Translation="0,0,0", Rotation="-5,0,0", Scale="1,1,1") {
    rotationMode=BMM_Additive
    rotationSpace=BCS_BoneSpace
    scaleMode=BMM_Additive
    scaleSpace=BCS_BoneSpace
    translationMode=BMM_Additive
    translationSpace=BCS_BoneSpace
    alphaBoolBlend {blendOption=Linear}
    alphaScaleBiasClamp {interpSpeedDecreasing=5}
    boneToModify {boneName=dyn_hair_bck_ribbon_r_05}
    rotation {pitch=-5}
   }
   n40 = "Transform (Modify) Bone - Bone: dyn_hair_bck_ribbon_l_base"(ComponentPose=@n39, Alpha=@n37, Translation="0,6,0", Rotation="0,-5,0", Scale="1,0.8,1") {
    rotationMode=BMM_Additive
    rotationSpace=BCS_BoneSpace
    scaleMode=BMM_Additive
    scaleSpace=BCS_BoneSpace
    translationMode=BMM_Additive
    translationSpace=BCS_BoneSpace
    alphaBoolBlend {blendOption=Linear}
    alphaScaleBiasClamp {interpSpeedDecreasing=5}
    boneToModify {boneName=dyn_hair_bck_ribbon_l_base}
    rotation {yaw=-5}
    scale {y=0.8}
    translation {y=6}
   }
   n41 = "Transform (Modify) Bone - Bone: dyn_hair_bck_ribbon_r_base"(ComponentPose=@n40, Alpha=@n37, Translation="0,2,0", Rotation="0,5,0", Scale="1,0.8,1") {
    rotationMode=BMM_Additive
    rotationSpace=BCS_BoneSpace
    scaleMode=BMM_Additive
    scaleSpace=BCS_BoneSpace
    translationMode=BMM_Additive
    translationSpace=BCS_BoneSpace
    alphaBoolBlend {blendOption=Linear}
    alphaScaleBiasClamp {bInterpResult=true, interpSpeedDecreasing=5}
    boneToModify {boneName=dyn_hair_bck_ribbon_r_base}
    rotation {yaw=5}
    scale {y=0.8}
    translation {y=2}
   }
   n42 = "Transform (Modify) Bone - Bone: dyn_hair_bck_braid_base"(ComponentPose=@n41, Alpha=1, Translation="0,0,0", Rotation="0,5,0", Scale="1,0.8,1") {
    rotationMode=BMM_Additive
    rotationSpace=BCS_BoneSpace
    scaleMode=BMM_Additive
    scaleSpace=BCS_BoneSpace
    translationMode=BMM_Additive
    translationSpace=BCS_BoneSpace
    alphaBoolBlend {blendOption=Linear}
    alphaScaleBiasClamp {interpSpeedDecreasing=5}
    boneToModify {boneName=dyn_hair_bck_braid_base}
    rotation {yaw=5}
    scale {y=0.8}
   }
   n43 = "Transform (Modify) Bone - Bone: dyn_hair_bck_ribbon_l_base"(ComponentPose=@n42, Alpha=1, Translation="1,0,0", Rotation="0,0,0", Scale="1,1,1") {
    rotationMode=BMM_Additive
    rotationSpace=BCS_BoneSpace
    scaleMode=BMM_Additive
    scaleSpace=BCS_BoneSpace
    translationMode=BMM_Additive
    translationSpace=BCS_BoneSpace
    alphaBoolBlend {blendOption=Linear}
    alphaScaleBiasClamp {interpSpeedDecreasing=5}
    boneToModify {boneName=dyn_hair_bck_ribbon_l_base}
    translation {x=1}
   }
   n44 = "Transform (Modify) Bone - Bone: dyn_hair_bck_ribbon_r_base"!(ComponentPose=@n43, bAlphaBoolEnabled:bool=True, Translation="-1,0,0", Rotation="2.5,0,0", Scale="1,1,1") -> (Pose) {
    alphaInputType=Bool
    rotationMode=BMM_Additive
    rotationSpace=BCS_BoneSpace
    scaleMode=BMM_Additive
    scaleSpace=BCS_BoneSpace
    translationMode=BMM_Additive
    translationSpace=BCS_BoneSpace
    alphaBoolBlend {blendOption=Linear}
    alphaScaleBiasClamp {bInterpResult=true, interpSpeedDecreasing=5}
    boneToModify {boneName=dyn_hair_bck_ribbon_r_base}
    rotation {pitch=2.5}
    translation {x=-1}
   }
   n45 = "Transform (Modify) Bone - Bone: dyn_hair_side_l_base"!(ComponentPose=@n44, bAlphaBoolEnabled:bool=True, Translation="0,0,0", Rotation="10,-10,0", Scale="1,1,1") -> (Pose) {
    alphaInputType=Bool
    rotationMode=BMM_Additive
    rotationSpace=BCS_BoneSpace
    translationMode=BMM_Additive
    translationSpace=BCS_BoneSpace
    alphaBoolBlend {blendOption=Sinusoidal}
    alphaScaleBiasClamp {bInterpResult=true, interpSpeedDecreasing=5}
    boneToModify {boneName=dyn_hair_side_l_base}
    rotation {pitch=10, yaw=-10}
   }
   n46 = "Transform (Modify) Bone - Bone: dyn_hair_bck_ribbon_l_base"!(ComponentPose=@n45, bAlphaBoolEnabled:bool=True, Translation="0,0,0", Rotation="5,0,0", Scale="1,1,1") -> (Pose) {
    alphaInputType=Bool
    rotationMode=BMM_Additive
    rotationSpace=BCS_BoneSpace
    scaleMode=BMM_Additive
    scaleSpace=BCS_BoneSpace
    translationMode=BMM_Additive
    translationSpace=BCS_BoneSpace
    alphaBoolBlend {blendOption=CubicInOut}
    alphaScaleBiasClamp {bInterpResult=true, interpSpeedDecreasing=5}
    boneToModify {boneName=dyn_hair_bck_ribbon_l_base}
    rotation {pitch=5}
   }
   n47 = "Transform (Modify) Bone - Bone: dyn_hair_bck_ribbon_r_base"(ComponentPose=@n46, bAlphaBoolEnabled=True, Translation="0,0,0", Rotation="5,0,0", Scale="1,1,1") {
    alphaInputType=Bool
    rotationMode=BMM_Additive
    rotationSpace=BCS_BoneSpace
    scaleMode=BMM_Additive
    scaleSpace=BCS_BoneSpace
    translationMode=BMM_Additive
    translationSpace=BCS_BoneSpace
    alphaBoolBlend {blendOption=Sinusoidal}
    alphaScaleBiasClamp {bInterpResult=true, interpSpeedDecreasing=5}
    boneToModify {boneName=dyn_hair_bck_ribbon_r_base}
    rotation {pitch=5}
   }
   n48 = "Transform (Modify) Bone - Bone: dyn_hair_bck_braid_base"!(ComponentPose=@n47, bAlphaBoolEnabled:bool=True, Translation="0,0,0", Rotation="5,0,0", Scale="1,1,1") -> (Pose) {
    alphaInputType=Bool
    rotationMode=BMM_Additive
    rotationSpace=BCS_BoneSpace
    scaleMode=BMM_Additive
    scaleSpace=BCS_BoneSpace
    translationMode=BMM_Additive
    translationSpace=BCS_BoneSpace
    alphaBoolBlend {blendOption=Sinusoidal}
    alphaScaleBiasClamp {bInterpResult=true, interpSpeedDecreasing=5}
    boneToModify {boneName=dyn_hair_bck_braid_base}
    rotation {pitch=5}
   }
   Save cached pose 'Adjusments 02'(Pose=Component To Local(ComponentPose=@n48)) {cacheName=Adjusments 02}
   n49 = Get rotatorWaveRope_Y() -> (rotatorWaveRope_Y:Rotator)
   n50 = "Transform (Modify) Bone - Bone: dyn_hair_fr_r_01"(ComponentPose=Local To Component(LocalPose=Use cached pose 'Adjusments 02'()), Alpha=1, Translation="0,0,0", Rotation=@n49, Scale="1,1,1") {
    rotationMode=BMM_Additive
    rotationSpace=BCS_BoneSpace
    translationMode=BMM_Additive
    translationSpace=BCS_BoneSpace
    alphaBoolBlend {blendOption=Linear}
    alphaScaleBiasClamp {bInterpResult=true, interpSpeedDecreasing=5, interpSpeedIncreasing=5}
    boneToModify {boneName=dyn_hair_fr_r_01}
   }
   n51 = "Transform (Modify) Bone - Bone: dyn_hair_fr_l_01"(ComponentPose=@n50, Alpha=1, Translation="0,0,0", Rotation=@n49, Scale="1,1,1") {
    rotationMode=BMM_Additive
    rotationSpace=BCS_BoneSpace
    translationMode=BMM_Additive
    translationSpace=BCS_BoneSpace
    alphaBoolBlend {blendOption=Linear}
    alphaScaleBiasClamp {bInterpResult=true, interpSpeedDecreasing=5, interpSpeedIncreasing=5}
    boneToModify {boneName=dyn_hair_fr_l_01}
   }
   n52 = "Transform (Modify) Bone - Bone: dyn_hair_pigtail_r_01"(ComponentPose=@n51, Alpha=1, Translation="0,0,0", Rotation=@n49, Scale="1,1,1") {
    rotationMode=BMM_Additive
    rotationSpace=BCS_BoneSpace
    translationMode=BMM_Additive
    translationSpace=BCS_BoneSpace
    alphaBoolBlend {blendOption=Linear}
    alphaScaleBiasClamp {bInterpResult=true, interpSpeedDecreasing=5, interpSpeedIncreasing=5}
    boneToModify {boneName=dyn_hair_pigtail_r_01}
   }
   n53 = "Transform (Modify) Bone - Bone: dyn_hair_ribbon_A_01"(ComponentPose=@n52, Alpha=1, Translation="0,0,0", Rotation=@n49, Scale="1,1,1") {
    rotationMode=BMM_Additive
    rotationSpace=BCS_BoneSpace
    translationMode=BMM_Additive
    translationSpace=BCS_BoneSpace
    alphaBoolBlend {blendOption=Linear}
    alphaScaleBiasClamp {bInterpResult=true, interpSpeedDecreasing=5, interpSpeedIncreasing=5}
    boneToModify {boneName=dyn_hair_ribbon_A_01}
   }
   Save cached pose 'All_sim_Cache'(Pose=Component To Local(ComponentPose=@n53)) {cacheName=All_sim_Cache}
   n54 = Get headUp()
   n55 = "Transform (Modify) Bone - Bone: dyn_hair_fr_l_01"(
    ComponentPose=Local To Component(LocalPose=Use cached pose 'CtrlRigCache'())
    Alpha=@n54
    Translation="7.5,0,0"
    Rotation="0,0,0"
    Scale="1,1,1"
   ) {
    rotationMode=BMM_Additive
    rotationSpace=BCS_BoneSpace
    translationMode=BMM_Additive
    translationSpace=BCS_BoneSpace
    alphaBoolBlend {blendOption=Linear}
    alphaScaleBiasClamp {interpSpeedDecreasing=5}
    boneToModify {boneName=dyn_hair_fr_l_01}
    translation {x=7.5}
   }
   n56 = "Transform (Modify) Bone - Bone: dyn_hair_side_r_base"(ComponentPose=@n55, Alpha=@n54, Translation="-4,-2,2", Rotation="0,0,0", Scale="1,1,1") {
    rotationMode=BMM_Additive
    rotationSpace=BCS_BoneSpace
    scaleMode=BMM_Additive
    scaleSpace=BCS_BoneSpace
    translationMode=BMM_Additive
    translationSpace=BCS_BoneSpace
    alphaBoolBlend {blendOption=Linear}
    alphaScaleBiasClamp {interpSpeedDecreasing=5}
    boneToModify {boneName=dyn_hair_side_r_base}
    translation=<-4 -2 2>
   }
   Save cached pose 'PreSim_Cache'(Pose=Component To Local(ComponentPose=@n56)) {cacheName=PreSim_Cache}
   n57 = "Constraint - dyn_hair_bck_l_01 (PP)"(ComponentPose=Local To Component(LocalPose=Use cached pose 'PreSim_Cache'()), Alpha=0.4, ConstraintWeights_0=0.5, ConstraintWeights_1=0.5) {
    alphaBoolBlend {blendOption=Linear}
    boneToModify {boneName=dyn_hair_bck_l_01}
    constraintSetup [{targetBone {boneName=spine_05}}, {targetBone {boneName=clavicle_l}}]
    constraintWeights [1,1]
   }
   n58 = "Constraint - dyn_hair_bck_l_02 (PP)"(ComponentPose=@n57, Alpha=0.8, ConstraintWeights_0=0.8, ConstraintWeights_1=0.2) {
    alphaBoolBlend {blendOption=Linear}
    boneToModify {boneName=dyn_hair_bck_l_02}
    constraintSetup [{targetBone {boneName=spine_05}}, {targetBone {boneName=clavicle_l}}]
    constraintWeights [1,1]
   }
   n59 = "Constraint - dyn_hair_side_l_01 (PP)"(ComponentPose=@n58, Alpha=0.5, ConstraintWeights_0=0.8, ConstraintWeights_1=0.2) {
    alphaBoolBlend {blendOption=Linear}
    boneToModify {boneName=dyn_hair_side_l_01}
    constraintSetup [{targetBone {boneName=spine_05}}, {targetBone {boneName=clavicle_l}}]
    constraintWeights [1,1]
   }
   n60 = "Constraint - dyn_hair_side_l_02 (PP)"(ComponentPose=@n59, Alpha=0.5, ConstraintWeights_0=0.2, ConstraintWeights_1=0.8) {
    alphaBoolBlend {blendOption=Linear}
    boneToModify {boneName=dyn_hair_side_l_02}
    constraintSetup [{targetBone {boneName=spine_05}}, {targetBone {boneName=clavicle_l}}]
    constraintWeights [1,1]
   }
   n61 = "Constraint - dyn_hair_fr_l_01 (PP)"!(ComponentPose=@n60, Alpha=0.2, ConstraintWeights_0=0.2, ConstraintWeights_1:float=0.8) -> (Pose) {
    alphaBoolBlend {blendOption=Linear}
    boneToModify {boneName=dyn_hair_fr_l_01}
    constraintSetup [{targetBone {boneName=spine_05}}, {targetBone {boneName=clavicle_l}}]
    constraintWeights [1,1]
   }
   n62 = "Constraint - dyn_hair_bck_r_03 (PP)"(ComponentPose=@n61, Alpha=1, ConstraintWeights_0=0.8, ConstraintWeights_1=0.2) {
    alphaBoolBlend {blendOption=Linear}
    boneToModify {boneName=dyn_hair_bck_r_03}
    constraintSetup [{targetBone {boneName=clavicle_r}}, {targetBone {boneName=spine_05}}]
    constraintWeights [1,1]
   }
   n63 = "Constraint - dyn_hair_side_r_01 (PP)"(ComponentPose=@n62, Alpha=0.5, ConstraintWeights_0=0.8, ConstraintWeights_1=0.2) {
    alphaBoolBlend {blendOption=Linear}
    boneToModify {boneName=dyn_hair_side_r_01}
    constraintSetup [{targetBone {boneName=spine_05}}, {targetBone {boneName=clavicle_r}}]
    constraintWeights [1,1]
   }
   n64 = "Constraint - dyn_hair_side_r_02 (PP)"(ComponentPose=@n63, Alpha=0.8, ConstraintWeights_0=0.2, ConstraintWeights_1=0.8) {
    alphaBoolBlend {blendOption=Linear}
    boneToModify {boneName=dyn_hair_side_r_02}
    constraintSetup [{targetBone {boneName=spine_05}}, {targetBone {boneName=clavicle_r}}]
    constraintWeights [1,1]
   }
   n65 = "Constraint - dyn_hair_fr_r_01 (PP)"!(ComponentPose=@n64, Alpha=0.2, ConstraintWeights_0=0.8, ConstraintWeights_1:float=0.2) -> (Pose) {
    alphaBoolBlend {blendOption=Linear}
    boneToModify {boneName=dyn_hair_fr_r_01}
    constraintSetup [{targetBone {boneName=spine_05}}, {targetBone {boneName=clavicle_r}}]
    constraintWeights [1,1]
   }
   n66 = "Constraint - dyn_hair_fr_r_02 (PP)"!(ComponentPose=@n65, Alpha=0.2, ConstraintWeights_0=0.8, ConstraintWeights_1:float=0.2) -> (Pose) {
    alphaBoolBlend {blendOption=Linear}
    boneToModify {boneName=dyn_hair_fr_r_02}
    constraintSetup [{targetBone {boneName=spine_05}}, {targetBone {boneName=clavicle_r}}]
    constraintWeights [1,1]
   }
   n67 = RigidBody(ComponentPose=@n66, OverrideWorldGravity=Get GravityOverride()) {
    bClampLinearTranslationLimitToRefPose=true
    bOverrideWorldGravity=true
    overridePhysicsAsset=&M_MED_SpringBreakTrip_FaceAcc_Physics
    simulationSpace=BaseBoneSpace
    alphaBoolBlend {blendOption=Linear}
    baseBoneRef {boneName=neck_02}
    componentAppliedLinearAccClamp=<3000 3000 3000>
    componentLinearAccScale=<0.3 0.3 0.3>
    componentLinearVelScale=<0.3 0.3 0.3>
    overrideWorldGravity {z=-980}
    simSpaceSettings {maxLinearVelocity=0.3, worldAlpha=0.35}
   }
   n68 = Component To Local(ComponentPose=@n67)
   n69 = RigidBody(ComponentPose=Local To Component(LocalPose=Use cached pose 'PreSim_Cache'()), OverrideWorldGravity="0,0,-980") {
    bClampLinearTranslationLimitToRefPose=true
    bOverrideWorldGravity=true
    overridePhysicsAsset=&/M_MED_SpringBreakTrip_FaceAcc_Physics_Skydiving
    simulationSpace=BaseBoneSpace
    alphaBoolBlend {blendOption=Linear}
    baseBoneRef {boneName=spine_04}
    componentAppliedLinearAccClamp=<5000 5000 5000>
    componentLinearAccScale=<0.3 0.3 0.08>
    componentLinearVelScale=<0.3 0.3 0.08>
    overrideWorldGravity {z=-980}
   }
   n70 = Component To Local(ComponentPose=@n69)
   n71 = Get bIsSkyDiving()
   n72 = "Constraint - dyn_hair_bck_ribbon_l_01 (P)"(
    ComponentPose=Local To Component(LocalPose=Blend Poses by bool(BlendPose_0=@n70, BlendPose_1=@n68, bActiveValue=@n71) {blendType=Linear})
    Alpha=0.8
    ConstraintWeights_0=1
   ) {
    alphaBoolBlend {blendOption=Linear}
    boneToModify {boneName=dyn_hair_bck_ribbon_l_01}
    constraintSetup [{targetBone {boneName=dyn_hair_bck_r_02}}]
    constraintWeights [1]
   }
   n73 = "Constraint - dyn_hair_bck_ribbon_r_01 (PP)"(ComponentPose=@n72, Alpha=0.8, ConstraintWeights_0=0.5, ConstraintWeights_1=0.5) {
    alphaBoolBlend {blendOption=Linear}
    boneToModify {boneName=dyn_hair_bck_ribbon_r_01}
    constraintSetup [{targetBone {boneName=dyn_hair_side_r_01}}, {targetBone {boneName=dyn_hair_bck_r_02}}]
    constraintWeights [1,1]
   }
   n74 = "Constraint - dyn_hair_bck_ribbon_l_02 (R)"(ComponentPose=@n73, Alpha=0.8, ConstraintWeights_0=1) {
    alphaBoolBlend {blendOption=Linear}
    boneToModify {boneName=dyn_hair_bck_ribbon_l_02}
    constraintSetup [{transformType=Rotation, targetBone {boneName=dyn_hair_bck_r_04}}]
    constraintWeights [1]
   }
   n75 = "Constraint - dyn_hair_bck_ribbon_r_02 (PP)"(ComponentPose=@n74, Alpha=0.8, ConstraintWeights_0=0.4, ConstraintWeights_1=0.6) {
    alphaBoolBlend {blendOption=Linear}
    boneToModify {boneName=dyn_hair_bck_ribbon_r_02}
    constraintSetup [{targetBone {boneName=dyn_hair_bck_r_03}}, {targetBone {boneName=dyn_hair_side_r_02}}]
    constraintWeights [1,1]
   }
   n76 = "Constraint - dyn_hair_bck_ribbon_l_03 (P)"(ComponentPose=@n75, Alpha=0.8, ConstraintWeights_0=1) {
    alphaBoolBlend {blendOption=Linear}
    boneToModify {boneName=dyn_hair_bck_ribbon_l_03}
    constraintSetup [{targetBone {boneName=dyn_hair_bck_r_04}}]
    constraintWeights [1]
   }
   n77 = "Constraint - dyn_hair_bck_ribbon_r_03 (PP)"(ComponentPose=@n76, Alpha=0.8, ConstraintWeights_0=0.6, ConstraintWeights_1=0.4) {
    alphaBoolBlend {blendOption=Linear}
    boneToModify {boneName=dyn_hair_bck_ribbon_r_03}
    constraintSetup [{targetBone {boneName=dyn_hair_bck_r_04}}, {targetBone {boneName=dyn_hair_side_r_03}}]
    constraintWeights [1,1]
   }
   n78 = "Constraint - dyn_hair_bck_ribbon_l_04 (P)"(ComponentPose=@n77, Alpha=0.8, ConstraintWeights_0=1) {
    alphaBoolBlend {blendOption=Linear}
    boneToModify {boneName=dyn_hair_bck_ribbon_l_04}
    constraintSetup [{targetBone {boneName=dyn_hair_bck_r_05}}]
    constraintWeights [1]
   }
   n79 = "Constraint - dyn_hair_bck_ribbon_r_04 (P)"(ComponentPose=@n78, Alpha=0.8, ConstraintWeights_0=1) {
    alphaBoolBlend {blendOption=Linear}
    boneToModify {boneName=dyn_hair_bck_ribbon_r_04}
    constraintSetup [{targetBone {boneName=dyn_hair_bck_r_05}}]
    constraintWeights [1]
   }
   n80 = "Constraint - dyn_hair_bck_ribbon_r_02 (T)"(ComponentPose=@n79, Alpha=0.15, ConstraintWeights_0=1) {
    alphaBoolBlend {blendOption=Linear}
    boneToModify {boneName=dyn_hair_bck_ribbon_r_02}
    constraintSetup [{transformType=Translation, targetBone {boneName=dyn_hair_side_r_03}}]
    constraintWeights [1]
   }
   n81 = "Constraint - dyn_hair_beads_l_01 (P)"(ComponentPose=@n80, Alpha=0.5, ConstraintWeights_0=1) {
    alphaBoolBlend {blendOption=Linear}
    boneToModify {boneName=dyn_hair_beads_l_01}
    constraintSetup [{targetBone {boneName=dyn_hair_fr_l_02}}]
    constraintWeights [1]
   }
   n82 = "Constraint - dyn_hair_beads_l_02 (PP)"!(ComponentPose=@n81, Alpha=0.8, ConstraintWeights_0=0.7, ConstraintWeights_1:float=0.3) -> (Pose) {
    alphaBoolBlend {blendOption=Linear}
    boneToModify {boneName=dyn_hair_beads_l_02}
    constraintSetup [{targetBone {boneName=dyn_hair_fr_l_03}}, {targetBone {boneName=dyn_hair_fr_l_04}}]
    constraintWeights [1,1]
   }
   n83 = "Constraint - dyn_hair_bck_braid_01 (PP)"(ComponentPose=@n82, Alpha=0.3, ConstraintWeights_0=0.75, ConstraintWeights_1=0.25) {
    alphaBoolBlend {blendOption=Linear}
    boneToModify {boneName=dyn_hair_bck_braid_01}
    constraintSetup [{targetBone {boneName=dyn_hair_bck_l_01}}, {targetBone {boneName=dyn_hair_bck_l_02}}]
    constraintWeights [1,1]
   }
   n84 = "Constraint - dyn_hair_bck_braid_02 (P)"(ComponentPose=@n83, Alpha=0.5, ConstraintWeights_0=1) {
    alphaBoolBlend {blendOption=Linear}
    boneToModify {boneName=dyn_hair_bck_braid_02}
    constraintSetup [{targetBone {boneName=dyn_hair_bck_l_02}}]
    constraintWeights [1]
   }
   n85 = "Constraint - dyn_hair_bck_braid_03 (R)"(ComponentPose=@n84, Alpha=1, ConstraintWeights_0=1) {
    alphaBoolBlend {blendOption=Linear}
    boneToModify {boneName=dyn_hair_bck_braid_03}
    constraintSetup [{transformType=Rotation, targetBone {boneName=dyn_hair_bck_l_03}}]
    constraintWeights [1]
   }
   n86 = "Constraint - dyn_hair_ribbon_A_02 (P)"(ComponentPose=@n85, Alpha=0, ConstraintWeights_0=1) {
    alphaBoolBlend {blendOption=Linear}
    boneToModify {boneName=dyn_hair_ribbon_A_02}
    constraintSetup [{targetBone {boneName=dyn_hair_fr_r_01}}]
    constraintWeights [1]
   }
   n87 = "Constraint - dyn_hair_ribbon_A_03 (P)"(ComponentPose=@n86, Alpha=0, ConstraintWeights_0=1) {
    alphaBoolBlend {blendOption=Linear}
    boneToModify {boneName=dyn_hair_ribbon_A_03}
    constraintSetup [{targetBone {boneName=dyn_hair_fr_r_02}}]
    constraintWeights [1]
   }
   n88 = "Constraint - dyn_hair_ribbon_B_01 (P)"(ComponentPose=@n87, Alpha=0, ConstraintWeights_0=1) {
    alphaBoolBlend {blendOption=Linear}
    boneToModify {boneName=dyn_hair_ribbon_B_01}
    constraintSetup [{targetBone {boneName=dyn_hair_ribbon_A_01}}]
    constraintWeights [1]
   }
   n89 = "Constraint - dyn_hair_ribbon_B_02 (P)"(ComponentPose=@n88, Alpha=0, ConstraintWeights_0=1) {
    alphaBoolBlend {blendOption=Linear}
    boneToModify {boneName=dyn_hair_ribbon_B_02}
    constraintSetup [{targetBone {boneName=dyn_hair_ribbon_A_02}}]
    constraintWeights [1]
   }
   n90 = "Constraint - dyn_hair_ribbon_B_03 (P)"(ComponentPose=@n89, Alpha=0, ConstraintWeights_0=1) {
    alphaBoolBlend {blendOption=Linear}
    boneToModify {boneName=dyn_hair_ribbon_B_03}
    constraintSetup [{targetBone {boneName=dyn_hair_ribbon_A_03}}]
    constraintWeights [1]
   }
   n91 = "Constraint - dyn_hair_ribbon_C_01 (P)"(ComponentPose=@n90, Alpha=1, ConstraintWeights_0=1) {
    alphaBoolBlend {blendOption=Linear}
    boneToModify {boneName=dyn_hair_ribbon_C_01}
    constraintSetup [{targetBone {boneName=dyn_hair_fr_r_01}}]
    constraintWeights [1]
   }
   n92 = "Constraint - dyn_hair_ribbon_C_02 (P)"(ComponentPose=@n91, Alpha=1, ConstraintWeights_0=1) {
    alphaBoolBlend {blendOption=Linear}
    boneToModify {boneName=dyn_hair_ribbon_C_02}
    constraintSetup [{targetBone {boneName=dyn_hair_fr_r_02}}]
    constraintWeights [1]
   }
   n93 = "Constraint - dyn_hair_ribbon_C_03 (P)"(ComponentPose=@n92, Alpha=0.5, ConstraintWeights_0=1) {
    alphaBoolBlend {blendOption=Linear}
    boneToModify {boneName=dyn_hair_ribbon_C_03}
    constraintSetup [{targetBone {boneName=dyn_hair_fr_r_03}}]
    constraintWeights [1]
   }
   n94 = Anim Dynamics - Bone: dyn_hair_dread_fr_l(ComponentPose=@n93, Alpha=0.5, ExternalForce="0,0,0") {
    angularDampingOverride=1
    angularSpringConstant=5500
    bAngularSpring=true
    bOverrideAngularDamping=true
    bUseGravityOverride=true
    gravityScale=0.35
    alphaBoolBlend {blendOption=Linear}
    boundBone {boneName=dyn_hair_dread_fr_l}
    componentLinearAccScale=<0.5 0.5 0.5>
    componentLinearVelScale=<0.5 0.5 0.5>
    gravityOverride {z=-980}
    physicsBodyDefinitions [{
      boundBone {boneName=dyn_hair_dread_fr_l}
      boxExtents=<22 22 22>
      constraintSetup {angularLimitsMax=<2 0 2>, angularLimitsMin=<-2 -40 -2>, angularTarget=<1 0 0>}
      localJointOffset=<2 0 0>
     }]
    relativeSpaceBone {boneName=root}
   }
   n95 = Anim Dynamics - Bone: dyn_hair_beads_r_bot(ComponentPose=@n94, Alpha=1, ExternalForce="0,0,0") {
    angularDampingOverride=1
    angularSpringConstant=5500
    bAngularSpring=true
    bOverrideAngularDamping=true
    bUseGravityOverride=true
    gravityScale=0.35
    alphaBoolBlend {blendOption=Linear}
    boundBone {boneName=dyn_hair_beads_r_bot}
    componentLinearAccScale=<0.5 0.5 0.5>
    componentLinearVelScale=<0.5 0.5 0.5>
    gravityOverride {z=-980}
    physicsBodyDefinitions [{
      boundBone {boneName=dyn_hair_beads_r_bot}
      boxExtents=<22 22 22>
      constraintSetup {angularLimitsMax=<2 0 2>, angularLimitsMin=<-2 -40 -2>, angularTarget=<1 0 0>}
      localJointOffset=<2 0 0>
     }]
    relativeSpaceBone {boneName=root}
   }
   n96 = Anim Dynamics - Bone: dyn_hair_beads_r_top_01(ComponentPose=@n95, Alpha=0.5, ExternalForce="0,0,0") {
    angularDampingOverride=1
    angularSpringConstant=5500
    bAngularSpring=true
    bOverrideAngularDamping=true
    bUseGravityOverride=true
    gravityScale=0.35
    alphaBoolBlend {blendOption=Linear}
    boundBone {boneName=dyn_hair_beads_r_top_01}
    componentLinearAccScale=<0.5 0.5 0.5>
    componentLinearVelScale=<0.5 0.5 0.5>
    gravityOverride {z=-980}
    physicsBodyDefinitions [{
      boundBone {boneName=dyn_hair_beads_r_top_01}
      boxExtents=<24 24 24>
      constraintSetup {angularLimitsMax=<2 0 2>, angularLimitsMin=<-2 -40 -2>, angularTarget=<1 0 0>}
      localJointOffset=<2 0 0>
     }]
    relativeSpaceBone {boneName=root}
   }
   n97 = Anim Dynamics - Bone: dyn_hair_beads_r_top_02(ComponentPose=@n96, Alpha=1, ExternalForce="0,0,0") {
    angularDampingOverride=1
    angularSpringConstant=5500
    bAngularSpring=true
    bOverrideAngularDamping=true
    bUseGravityOverride=true
    gravityScale=0.35
    alphaBoolBlend {blendOption=Linear}
    boundBone {boneName=dyn_hair_beads_r_top_02}
    componentLinearAccScale=<0.5 0.5 0.5>
    componentLinearVelScale=<0.5 0.5 0.5>
    gravityOverride {z=-980}
    physicsBodyDefinitions [{
      boundBone {boneName=dyn_hair_beads_r_top_02}
      boxExtents=<24 24 24>
      constraintSetup {angularLimitsMax=<2 0 2>, angularLimitsMin=<-2 -40 -2>, angularTarget=<1 0 0>}
      localJointOffset=<2 0 0>
     }]
    relativeSpaceBone {boneName=root}
   }
   n98 = Anim Dynamics - Bone: dyn_hair_ribbon_A_01(ComponentPose=@n97, Alpha=0.5, ExternalForce="0,0,0") {
    angularDampingOverride=1
    angularSpringConstant=5500
    bAngularSpring=true
    bOverrideAngularDamping=true
    bUseGravityOverride=true
    gravityScale=0.35
    alphaBoolBlend {blendOption=Linear}
    boundBone {boneName=dyn_hair_ribbon_A_01}
    componentLinearAccScale=<0.5 0.5 0.5>
    componentLinearVelScale=<0.5 0.5 0.5>
    gravityOverride {z=-980}
    physicsBodyDefinitions [{
      boundBone {boneName=dyn_hair_ribbon_A_01}
      boxExtents=<20 30 30>
      constraintSetup {angularLimitsMax=<30 0 30>, angularLimitsMin=<-30 -60 -30>, angularTarget=<1 0 0>}
      localJointOffset=<4 0 0>
     }]
    relativeSpaceBone {boneName=root}
   }
   n99 = Anim Dynamics - Bone: dyn_hair_ribbon_A_02(ComponentPose=@n98, Alpha=1, ExternalForce="0,0,0") {
    angularDampingOverride=1
    angularSpringConstant=5500
    bAngularSpring=true
    bOverrideAngularDamping=true
    bUseGravityOverride=true
    gravityScale=0.35
    alphaBoolBlend {blendOption=Linear}
    boundBone {boneName=dyn_hair_ribbon_A_02}
    componentLinearAccScale=<0.5 0.5 0.5>
    componentLinearVelScale=<0.5 0.5 0.5>
    gravityOverride {z=-980}
    physicsBodyDefinitions [{
      boundBone {boneName=dyn_hair_ribbon_A_02}
      boxExtents=<20 30 30>
      constraintSetup {angularLimitsMax=<30 0 30>, angularLimitsMin=<-30 -50 -30>, angularTarget=<1 0 0>}
      localJointOffset=<4 0 0>
     }]
    relativeSpaceBone {boneName=root}
   }
   n100 = Anim Dynamics - Bone: dyn_hair_ribbon_A_03(ComponentPose=@n99, Alpha=1, ExternalForce="0,0,0") {
    angularDampingOverride=1
    angularSpringConstant=5500
    bAngularSpring=true
    bOverrideAngularDamping=true
    bUseGravityOverride=true
    gravityScale=0.35
    alphaBoolBlend {blendOption=Linear}
    boundBone {boneName=dyn_hair_ribbon_A_03}
    componentLinearAccScale=<0.5 0.5 0.5>
    componentLinearVelScale=<0.5 0.5 0.5>
    gravityOverride {z=-980}
    physicsBodyDefinitions [{
      boundBone {boneName=dyn_hair_ribbon_A_03}
      boxExtents=<20 30 30>
      constraintSetup {angularLimitsMax=<30 0 30>, angularLimitsMin=<-30 -40 -30>, angularTarget=<1 0 0>}
      localJointOffset=<4 0 0>
     }]
    relativeSpaceBone {boneName=root}
   }
   n101 = Anim Dynamics - Bone: dyn_hair_ribbon_B_01(ComponentPose=@n100, Alpha=0.5, ExternalForce="0,0,0") {
    angularDampingOverride=1
    angularSpringConstant=5500
    bAngularSpring=true
    bOverrideAngularDamping=true
    bUseGravityOverride=true
    gravityScale=0.35
    alphaBoolBlend {blendOption=Linear}
    boundBone {boneName=dyn_hair_ribbon_B_01}
    componentLinearAccScale=<0.5 0.5 0.5>
    componentLinearVelScale=<0.5 0.5 0.5>
    gravityOverride {z=-980}
    physicsBodyDefinitions [{
      boundBone {boneName=dyn_hair_ribbon_B_01}
      boxExtents=<20 30 30>
      constraintSetup {angularLimitsMax=<20 0 20>, angularLimitsMin=<-20 -40 -20>, angularTarget=<1 0 0>}
      localJointOffset=<4 0 0>
     }]
    relativeSpaceBone {boneName=root}
   }
   n102 = Anim Dynamics - Bone: dyn_hair_ribbon_B_02(ComponentPose=@n101, Alpha=1, ExternalForce="0,0,0") {
    angularDampingOverride=1
    angularSpringConstant=5500
    bAngularSpring=true
    bOverrideAngularDamping=true
    bUseGravityOverride=true
    gravityScale=0.35
    alphaBoolBlend {blendOption=Linear}
    boundBone {boneName=dyn_hair_ribbon_B_02}
    componentLinearAccScale=<0.5 0.5 0.5>
    componentLinearVelScale=<0.5 0.5 0.5>
    gravityOverride {z=-980}
    physicsBodyDefinitions [{
      boundBone {boneName=dyn_hair_ribbon_B_02}
      boxExtents=<20 30 30>
      constraintSetup {angularLimitsMax=<20 0 20>, angularLimitsMin=<-20 -40 -20>, angularTarget=<1 0 0>}
      localJointOffset=<4 0 0>
     }]
    relativeSpaceBone {boneName=root}
   }
   n103 = Anim Dynamics - Bone: dyn_hair_ribbon_B_03(ComponentPose=@n102, Alpha=1, ExternalForce="0,0,0") {
    angularDampingOverride=1
    angularSpringConstant=5500
    bAngularSpring=true
    bOverrideAngularDamping=true
    bUseGravityOverride=true
    gravityScale=0.35
    alphaBoolBlend {blendOption=Linear}
    boundBone {boneName=dyn_hair_ribbon_B_03}
    componentLinearAccScale=<0.5 0.5 0.5>
    componentLinearVelScale=<0.5 0.5 0.5>
    gravityOverride {z=-980}
    physicsBodyDefinitions [{
      boundBone {boneName=dyn_hair_ribbon_B_03}
      boxExtents=<20 30 30>
      constraintSetup {angularLimitsMax=<20 0 20>, angularLimitsMin=<-20 -40 -20>, angularTarget=<1 0 0>}
      localJointOffset=<4 0 0>
     }]
    relativeSpaceBone {boneName=root}
   }
   n104 = Anim Dynamics - Bone: dyn_hair_ribbon_C_01(ComponentPose=@n103, Alpha=0.5, ExternalForce="0,0,0") {
    angularDampingOverride=1
    angularSpringConstant=5500
    bAngularSpring=true
    bOverrideAngularDamping=true
    bUseGravityOverride=true
    gravityScale=0.35
    alphaBoolBlend {blendOption=Linear}
    boundBone {boneName=dyn_hair_ribbon_C_01}
    componentLinearAccScale=<0.5 0.5 0.5>
    componentLinearVelScale=<0.5 0.5 0.5>
    gravityOverride {z=-980}
    physicsBodyDefinitions [{
      boundBone {boneName=dyn_hair_ribbon_C_01}
      boxExtents=<15 15 15>
      constraintSetup {angularLimitsMax=<2 0 2>, angularLimitsMin=<-2 -40 -2>, angularTarget=<1 0 0>}
      localJointOffset=<2 0 0>
     }]
    relativeSpaceBone {boneName=root}
   }
   n105 = Anim Dynamics - Bone: dyn_hair_ribbon_C_02(ComponentPose=@n104, Alpha=1, ExternalForce="0,0,0") {
    angularDampingOverride=1
    angularSpringConstant=5500
    bAngularSpring=true
    bOverrideAngularDamping=true
    bUseGravityOverride=true
    gravityScale=0.35
    alphaBoolBlend {blendOption=Linear}
    boundBone {boneName=dyn_hair_ribbon_C_02}
    componentLinearAccScale=<0.5 0.5 0.5>
    componentLinearVelScale=<0.5 0.5 0.5>
    gravityOverride {z=-980}
    physicsBodyDefinitions [{
      boundBone {boneName=dyn_hair_ribbon_C_02}
      boxExtents=<15 15 15>
      constraintSetup {angularLimitsMax=<2 0 2>, angularLimitsMin=<-2 -40 -2>, angularTarget=<1 0 0>}
      localJointOffset=<2 0 0>
     }]
    relativeSpaceBone {boneName=root}
   }
   n106 = Anim Dynamics - Bone: dyn_hair_ribbon_C_03(ComponentPose=@n105, Alpha=1, ExternalForce="0,0,0") {
    angularDampingOverride=1
    angularSpringConstant=5500
    bAngularSpring=true
    bOverrideAngularDamping=true
    bUseGravityOverride=true
    gravityScale=0.35
    alphaBoolBlend {blendOption=Linear}
    boundBone {boneName=dyn_hair_ribbon_C_03}
    componentLinearAccScale=<0.5 0.5 0.5>
    componentLinearVelScale=<0.5 0.5 0.5>
    gravityOverride {z=-980}
    physicsBodyDefinitions [{
      boundBone {boneName=dyn_hair_ribbon_C_03}
      boxExtents=<15 15 15>
      constraintSetup {angularLimitsMax=<2 0 2>, angularLimitsMin=<-2 -40 -2>, angularTarget=<1 0 0>}
      localJointOffset=<2 0 0>
     }]
    relativeSpaceBone {boneName=root}
   }
   n107 = Anim Dynamics - Bone: dyn_hair_bck_braid_01(ComponentPose=@n106, Alpha=0.5, ExternalForce="0,0,0") {
    angularDampingOverride=1
    angularSpringConstant=5500
    bAngularSpring=true
    bOverrideAngularDamping=true
    bUseGravityOverride=true
    gravityScale=0.35
    alphaBoolBlend {blendOption=Linear}
    boundBone {boneName=dyn_hair_bck_braid_01}
    componentLinearAccScale=<0.5 0.5 0.5>
    componentLinearVelScale=<0.5 0.5 0.5>
    gravityOverride {z=-980}
    physicsBodyDefinitions [{
      boundBone {boneName=dyn_hair_bck_braid_01}
      boxExtents=<22 22 22>
      constraintSetup {angularLimitsMax=<2 0 2>, angularLimitsMin=<-2 -40 -2>, angularTarget=<1 0 0>}
      localJointOffset=<2 0 0>
     }]
    relativeSpaceBone {boneName=root}
   }
   n108 = Anim Dynamics - Bone: dyn_hair_bck_braid_02(ComponentPose=@n107, Alpha=1, ExternalForce="0,0,0") {
    angularDampingOverride=1
    angularSpringConstant=5500
    bAngularSpring=true
    bOverrideAngularDamping=true
    bUseGravityOverride=true
    gravityScale=0.35
    alphaBoolBlend {blendOption=Linear}
    boundBone {boneName=dyn_hair_bck_braid_02}
    componentLinearAccScale=<0.5 0.5 0.5>
    componentLinearVelScale=<0.5 0.5 0.5>
    gravityOverride {z=-980}
    physicsBodyDefinitions [{
      boundBone {boneName=dyn_hair_bck_braid_02}
      boxExtents=<22 22 22>
      constraintSetup {angularLimitsMax=<2 0 2>, angularLimitsMin=<-2 -40 -2>, angularTarget=<1 0 0>}
      localJointOffset=<2 0 0>
     }]
    relativeSpaceBone {boneName=root}
   }
   n109 = Anim Dynamics - Bone: dyn_hair_bck_braid_03(ComponentPose=@n108, Alpha=1, ExternalForce="0,0,0") {
    angularDampingOverride=1
    angularSpringConstant=5500
    bAngularSpring=true
    bOverrideAngularDamping=true
    bUseGravityOverride=true
    gravityScale=0.35
    alphaBoolBlend {blendOption=Linear}
    boundBone {boneName=dyn_hair_bck_braid_03}
    componentLinearAccScale=<0.5 0.5 0.5>
    componentLinearVelScale=<0.5 0.5 0.5>
    gravityOverride {z=-980}
    physicsBodyDefinitions [{
      boundBone {boneName=dyn_hair_bck_braid_03}
      boxExtents=<22 22 22>
      constraintSetup {angularLimitsMax=<2 0 2>, angularLimitsMin=<-2 -40 -2>, angularTarget=<1 0 0>}
      localJointOffset=<2 0 0>
     }]
    relativeSpaceBone {boneName=root}
   }
   n110 = Anim Dynamics - Bone: dyn_hair_bck_ribbon_l_01(ComponentPose=@n109, Alpha=1, ExternalForce="0,0,0") {
    angularDampingOverride=1
    angularSpringConstant=5500
    bAngularSpring=true
    bOverrideAngularDamping=true
    bUseGravityOverride=true
    gravityScale=0.35
    alphaBoolBlend {blendOption=Linear}
    boundBone {boneName=dyn_hair_bck_ribbon_l_01}
    componentLinearAccScale=<0.5 0.5 0.5>
    componentLinearVelScale=<0.5 0.5 0.5>
    gravityOverride {z=-980}
    physicsBodyDefinitions [{
      boundBone {boneName=dyn_hair_bck_ribbon_l_01}
      boxExtents=<20 20 20>
      constraintSetup {angularLimitsMax=<2 0 2>, angularLimitsMin=<-2 -10 0>, angularTarget=<1 0 0>}
      localJointOffset=<2 0 0>
     }]
    relativeSpaceBone {boneName=root}
   }
   n111 = Anim Dynamics - Bone: dyn_hair_bck_ribbon_l_02(ComponentPose=@n110, Alpha=1, ExternalForce="0,0,0") {
    angularDampingOverride=1
    angularSpringConstant=5500
    bAngularSpring=true
    bOverrideAngularDamping=true
    bUseGravityOverride=true
    gravityScale=0.35
    alphaBoolBlend {blendOption=Linear}
    boundBone {boneName=dyn_hair_bck_ribbon_l_02}
    componentLinearAccScale=<0.5 0.5 0.5>
    componentLinearVelScale=<0.5 0.5 0.5>
    gravityOverride {z=-980}
    physicsBodyDefinitions [{
      boundBone {boneName=dyn_hair_bck_ribbon_l_02}
      boxExtents=<20 20 20>
      constraintSetup {angularLimitsMax=<2 0 2>, angularLimitsMin=<-2 -10 0>, angularTarget=<1 0 0>}
      localJointOffset=<2 0 0>
     }]
    relativeSpaceBone {boneName=root}
   }
   n112 = Anim Dynamics - Bone: dyn_hair_bck_ribbon_l_03(ComponentPose=@n111, Alpha=1, ExternalForce="0,0,0") {
    angularDampingOverride=1
    angularSpringConstant=5500
    bAngularSpring=true
    bOverrideAngularDamping=true
    bUseGravityOverride=true
    gravityScale=0.35
    alphaBoolBlend {blendOption=Linear}
    boundBone {boneName=dyn_hair_bck_ribbon_l_03}
    componentLinearAccScale=<0.5 0.5 0.5>
    componentLinearVelScale=<0.5 0.5 0.5>
    gravityOverride {z=-980}
    physicsBodyDefinitions [{
      boundBone {boneName=dyn_hair_bck_ribbon_l_03}
      boxExtents=<20 20 20>
      constraintSetup {angularLimitsMax=<2 0 2>, angularLimitsMin=<-2 -10 0>, angularTarget=<1 0 0>}
      localJointOffset=<2 0 0>
     }]
    relativeSpaceBone {boneName=root}
   }
   n113 = Anim Dynamics - Bone: dyn_hair_bck_ribbon_l_04(ComponentPose=@n112, Alpha=1, ExternalForce="0,0,0") {
    angularDampingOverride=1
    angularSpringConstant=5500
    bAngularSpring=true
    bOverrideAngularDamping=true
    bUseGravityOverride=true
    gravityScale=0.35
    alphaBoolBlend {blendOption=Linear}
    boundBone {boneName=dyn_hair_bck_ribbon_l_04}
    componentLinearAccScale=<0.5 0.5 0.5>
    componentLinearVelScale=<0.5 0.5 0.5>
    gravityOverride {z=-980}
    physicsBodyDefinitions [{
      boundBone {boneName=dyn_hair_bck_ribbon_l_04}
      boxExtents=<20 20 20>
      constraintSetup {angularLimitsMax=<2 0 2>, angularLimitsMin=<-2 -10 0>, angularTarget=<1 0 0>}
      localJointOffset=<2 0 0>
     }]
    relativeSpaceBone {boneName=root}
   }
   n114 = Anim Dynamics - Bone: dyn_hair_bck_ribbon_l_05(ComponentPose=@n113, Alpha=1, ExternalForce="0,0,0") {
    angularDampingOverride=1
    angularSpringConstant=5500
    bAngularSpring=true
    bOverrideAngularDamping=true
    bUseGravityOverride=true
    gravityScale=0.35
    alphaBoolBlend {blendOption=Linear}
    boundBone {boneName=dyn_hair_bck_ribbon_l_05}
    componentLinearAccScale=<0.5 0.5 0.5>
    componentLinearVelScale=<0.5 0.5 0.5>
    gravityOverride {z=-980}
    physicsBodyDefinitions [{
      boundBone {boneName=dyn_hair_bck_ribbon_l_05}
      boxExtents=<20 20 20>
      constraintSetup {angularLimitsMax=<2 0 2>, angularLimitsMin=<-2 -10 0>, angularTarget=<1 0 0>}
      localJointOffset=<2 0 0>
     }]
    relativeSpaceBone {boneName=root}
   }
   n115 = Anim Dynamics - Bone: dyn_hair_bck_ribbon_r_01(ComponentPose=@n114, Alpha=1, ExternalForce="0,0,0") {
    angularDampingOverride=1
    angularSpringConstant=5500
    bAngularSpring=true
    bOverrideAngularDamping=true
    bUseGravityOverride=true
    gravityScale=0.35
    alphaBoolBlend {blendOption=Linear}
    boundBone {boneName=dyn_hair_bck_ribbon_r_01}
    componentLinearAccScale=<0.5 0.5 0.5>
    componentLinearVelScale=<0.5 0.5 0.5>
    gravityOverride {z=-980}
    physicsBodyDefinitions [{
      boundBone {boneName=dyn_hair_bck_ribbon_r_01}
      boxExtents=<18 18 18>
      constraintSetup {angularLimitsMax=<2 0 0>, angularLimitsMin=<-2 -10 -4>, angularTarget=<1 0 0>}
      localJointOffset=<2 0 0>
     }]
    relativeSpaceBone {boneName=root}
   }
   n116 = Anim Dynamics - Bone: dyn_hair_bck_ribbon_r_02(ComponentPose=@n115, Alpha=1, ExternalForce="0,0,0") {
    angularDampingOverride=1
    angularSpringConstant=5500
    bAngularSpring=true
    bOverrideAngularDamping=true
    bUseGravityOverride=true
    gravityScale=0.35
    alphaBoolBlend {blendOption=Linear}
    boundBone {boneName=dyn_hair_bck_ribbon_r_02}
    componentLinearAccScale=<0.5 0.5 0.5>
    componentLinearVelScale=<0.5 0.5 0.5>
    gravityOverride {z=-980}
    physicsBodyDefinitions [{
      boundBone {boneName=dyn_hair_bck_ribbon_r_02}
      boxExtents=<18 18 18>
      constraintSetup {angularLimitsMax=<2 0 0>, angularLimitsMin=<-2 -10 -4>, angularTarget=<1 0 0>}
      localJointOffset=<2 0 0>
     }]
    relativeSpaceBone {boneName=root}
   }
   n117 = Anim Dynamics - Bone: dyn_hair_bck_ribbon_r_03(ComponentPose=@n116, Alpha=1, ExternalForce="0,0,0") {
    angularDampingOverride=1
    angularSpringConstant=5500
    bAngularSpring=true
    bOverrideAngularDamping=true
    bUseGravityOverride=true
    gravityScale=0.35
    alphaBoolBlend {blendOption=Linear}
    boundBone {boneName=dyn_hair_bck_ribbon_r_03}
    componentLinearAccScale=<0.5 0.5 0.5>
    componentLinearVelScale=<0.5 0.5 0.5>
    gravityOverride {z=-980}
    physicsBodyDefinitions [{
      boundBone {boneName=dyn_hair_bck_ribbon_r_03}
      boxExtents=<18 18 18>
      constraintSetup {angularLimitsMax=<2 0 0>, angularLimitsMin=<-2 -10 -4>, angularTarget=<1 0 0>}
      localJointOffset=<2 0 0>
     }]
    relativeSpaceBone {boneName=root}
   }
   n118 = Anim Dynamics - Bone: dyn_hair_bck_ribbon_r_04(ComponentPose=@n117, Alpha=1, ExternalForce="0,0,0") {
    angularDampingOverride=1
    angularSpringConstant=5500
    bAngularSpring=true
    bOverrideAngularDamping=true
    bUseGravityOverride=true
    gravityScale=0.35
    alphaBoolBlend {blendOption=Linear}
    boundBone {boneName=dyn_hair_bck_ribbon_r_04}
    componentLinearAccScale=<0.5 0.5 0.5>
    componentLinearVelScale=<0.5 0.5 0.5>
    gravityOverride {z=-980}
    physicsBodyDefinitions [{
      boundBone {boneName=dyn_hair_bck_ribbon_r_04}
      boxExtents=<18 18 18>
      constraintSetup {angularLimitsMax=<2 0 0>, angularLimitsMin=<-2 -10 -4>, angularTarget=<1 0 0>}
      localJointOffset=<2 0 0>
     }]
    relativeSpaceBone {boneName=root}
   }
   n119 = Anim Dynamics - Bone: dyn_hair_bck_ribbon_r_05(ComponentPose=@n118, Alpha=1, ExternalForce="0,0,0") {
    angularDampingOverride=1
    angularSpringConstant=5500
    bAngularSpring=true
    bOverrideAngularDamping=true
    bUseGravityOverride=true
    gravityScale=0.35
    alphaBoolBlend {blendOption=Linear}
    boundBone {boneName=dyn_hair_bck_ribbon_r_05}
    componentLinearAccScale=<0.5 0.5 0.5>
    componentLinearVelScale=<0.5 0.5 0.5>
    gravityOverride {z=-980}
    physicsBodyDefinitions [{
      boundBone {boneName=dyn_hair_bck_ribbon_r_05}
      boxExtents=<18 18 18>
      constraintSetup {angularLimitsMax=<2 0 0>, angularLimitsMin=<-2 -10 -4>, angularTarget=<1 0 0>}
      localJointOffset=<2 0 0>
     }]
    relativeSpaceBone {boneName=root}
   }
   n120 = Anim Dynamics - Bone: dyn_hair_pigtail_r_01(ComponentPose=@n119, Alpha=1, ExternalForce="0,0,0") {
    angularDampingOverride=1
    angularSpringConstant=20000
    bAngularSpring=true
    bOverrideAngularDamping=true
    bUseGravityOverride=true
    gravityScale=0.6
    alphaBoolBlend {blendOption=Linear}
    boundBone {boneName=dyn_hair_pigtail_r_01}
    componentLinearAccScale=<1 1 1>
    componentLinearVelScale=<1 1 1>
    gravityOverride {z=-980}
    physicsBodyDefinitions [{
      boundBone {boneName=dyn_hair_pigtail_r_01}
      boxExtents=<20 30 30>
      constraintSetup {angularLimitsMax=<20 0 20>, angularLimitsMin=<-20 -60 -20>, angularTarget=<1 0 0>}
      localJointOffset=<4 0 0>
     }]
    relativeSpaceBone {boneName=spine_03}
   }
   n121 = Anim Dynamics - Bone: dyn_hair_pigtail_r_02(ComponentPose=@n120, Alpha=1, ExternalForce="0,0,0") {
    angularDampingOverride=1
    angularSpringConstant=20000
    bAngularSpring=true
    bOverrideAngularDamping=true
    bUseGravityOverride=true
    gravityScale=0.6
    alphaBoolBlend {blendOption=Linear}
    boundBone {boneName=dyn_hair_pigtail_r_02}
    componentLinearAccScale=<1 1 1>
    componentLinearVelScale=<1 1 1>
    gravityOverride {z=-980}
    physicsBodyDefinitions [{
      boundBone {boneName=dyn_hair_pigtail_r_02}
      boxExtents=<20 30 30>
      constraintSetup {angularLimitsMax=<20 0 20>, angularLimitsMin=<-20 -50 -20>, angularTarget=<1 0 0>}
      localJointOffset=<4 0 0>
     }]
    relativeSpaceBone {boneName=spine_03}
   }
   n122 = Anim Dynamics - Bone: dyn_hair_pigtail_r_03(ComponentPose=@n121, Alpha=1, ExternalForce="0,0,0") {
    angularDampingOverride=1
    angularSpringConstant=20000
    bAngularSpring=true
    bOverrideAngularDamping=true
    bUseGravityOverride=true
    gravityScale=0.6
    alphaBoolBlend {blendOption=Linear}
    boundBone {boneName=dyn_hair_pigtail_r_03}
    componentLinearAccScale=<1 1 1>
    componentLinearVelScale=<1 1 1>
    gravityOverride {z=-980}
    physicsBodyDefinitions [{
      boundBone {boneName=dyn_hair_pigtail_r_03}
      boxExtents=<20 30 30>
      constraintSetup {angularLimitsMax=<20 0 20>, angularLimitsMin=<-20 -40 -20>, angularTarget=<1 0 0>}
      localJointOffset=<4 0 0>
     }]
    relativeSpaceBone {boneName=spine_03}
   }
   n123 = Anim Dynamics - Bone: dyn_hair_dread_fr_l(ComponentPose=@n122, Alpha=1, ExternalForce="0,0,0") {
    angularDampingOverride=1
    angularSpringConstant=20000
    bAngularSpring=true
    bOverrideAngularDamping=true
    bUseGravityOverride=true
    gravityScale=0.35
    simulationSpace=RootRelative
    alphaBoolBlend {blendOption=Linear}
    boundBone {boneName=dyn_hair_dread_fr_l}
    componentLinearAccScale=<1 1 1>
    componentLinearVelScale=<1 1 1>
    gravityOverride {z=-980}
    physicsBodyDefinitions [{
      boundBone {boneName=dyn_hair_dread_fr_l}
      boxExtents=<20 25 25>
      constraintSetup {angularLimitsMax=<20 0 20>, angularLimitsMin=<-20 -40 -20>, angularTarget=<1 0 0>}
      localJointOffset=<4 0 0>
     }]
   }
   Save cached pose 'Constrain/Dyncamics Cache'(Pose=Component To Local(ComponentPose=@n123)) {cacheName=Constrain/Dyncamics Cache}
   Use cached pose 'Constrain/Dyncamics Cache'()
  ]}
 EventGraph {nodes [
   Event BlueprintUpdateAnimation()
   > n1 = Sequence!(execute) -> (then_0, then_1, then_2:Exec)
   > n2 = Gravity Override State Function DEPRECATED(CharacterPartAnimInstance=Self-Reference(), GameTimeInSeconds=Get Delta Time X(), GravityOverrideParams=Get Gravity Override Params())
   > Set GravityOverride(GravityOverride=@n2)
   n3 = CheckForBackbling()
   > Set isWearingBackbling(isWearingBackbling=AND Boolean!(A=@n3, B=NOT Boolean(A=@n3.bIsBackblingLong), C:bool=NOT Boolean(A=@n3.bIsBackblingLong)) -> (ReturnValue))
   > Set Is Wearing Cape(Is Wearing Cape=OR Boolean(A=@n3.bIsBackblingACape, B=@n3.bIsBackblingLong))
   > n4 = Branch(Condition=Get bIsSkydiveFloating())
   > Set speedWaveFur(speedWaveFur:double=5) -> (Output_Get:double)
   > Set amplitudWaveFur(amplitudWaveFur:double=1.7) -> (Output_Get:double)
   > Set offsetWaveFur(offsetWaveFur:double=0) -> (Output_Get:double)
   n5 = Branch(execute=@n4.else, Condition=Get bIsSkydiveDiveMode())
   > Set speedWaveFur(speedWaveFur=20)
   > Set amplitudWaveFur(amplitudWaveFur=0.5)
   > Set offsetWaveFur(offsetWaveFur=0)
   Set speedWaveFur(execute=@n5.else, speedWaveFur=0)
   Set headUp(
    execute=@n1.then_1
    headUp:double=GetAlphaFromReferenceTransformDelta(SkeletalMeshComponent=Get Body_Mesh(), BoneName=spine_05, BaseName=head, InputRange=(X=0,Y=90), TwistAxis="0,0,1")
   ) -> (Output_Get:double)
   n6 = "Sin (Radians)"(A=float * float(A=GetRealTimeSeconds(), B=Get speedWaveFur() -> (speedWaveFur:double)) {is=title K2Node_PromotableOperator})
   n7 = float + float(
    A=float * float(A=MultiplyByPi!(Value:double=@n6) -> (ReturnValue:double), B=Get amplitudWaveFur() -> (amplitudWaveFur:double))
    B=Get offsetWaveFur() -> (offsetWaveFur:double)
   )
   > Set rotatorWaveRope_Z(rotatorWaveRope_Z:Rotator=MakeRotator(Yaw=@n7)) -> (Output_Get:Rotator)
   > Set rotatorWaveRope_Y(rotatorWaveRope_Y:Rotator=MakeRotator(Pitch=@n7)) -> (Output_Get:Rotator)
   > Set isWearingBackCape(isWearingBackCape:bool=OR Boolean(A=Get isWearingBackbling(), B=Get Is Wearing Cape() -> (Is Wearing Cape:bool))) -> (Output_Get:bool)
   n8 = Event OnCharacterCustomizationLoaded()
   > Set Body_Mesh(Body_Mesh:Skeletal Mesh Component Budgeted Object Reference=Get Array Item with EFort Custom Part Type(Array=@n8.CustomMeshParts)) -> (Output_Get:Skeletal Mesh Component Budgeted Object Reference)
   Event OnCharacterAndPickaxeCustomizationLoaded!() -> (
    OutputDelegate
    then
    CustomBaseMesh:Skeletal Mesh Component Object Reference
    CustomParts:Array of Custom Character Part Object References
    CustomMeshParts:Array of Skeletal Mesh Component Budgeted Object References
   ) {is=title K2Node_Event}
  ]}
]
variables [
 amplitudWaveFur {type=number}
 Body_Mesh {type=SkeletalMeshComponentBudgeted}
 Delta Time X {type=number}
 Gravity Override Params {type=GravityOverrideParamsStruct}
 GravityOverride {type=Vector}
 headUp {type=number}
 Is Wearing Cape {type=boolean}
 isWearingBackbling {type=boolean}
 isWearingBackCape {type=boolean}
 offsetWaveFur {type=number}
 rotatorWaveRope_Y {type=Rotator}
 rotatorWaveRope_Z {type=Rotator}
 speedWaveFur {type=number}
]
